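; Dummy carrier actor for a pickup grabbed by the Gravy Goblin, letting the item be thrown and dropped on impact.
; AttachPickup stores the pickup and stops its movement; Throw sets Thrown on authority.
; A mesh hit while thrown makes authority detach the pickup just off the surface, register a sound indicator context, then destroy the dummy.
; OnRep_Thrown removes the GrabbedByGravyGoblin tag from the pickup.
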(event EventBeginPlay
  (|CallParentFunction) ; This node is disabled and will not be called.\nDrag off pins to build functionality.
)

(event Collision|EventActorBeginOverlap
  (params (OtherActor :type Actor%20Object%20Reference)))

(event Custom|Detach
  (params (NewLocation :type Vector))
  (bind _attachedpickup (Variables|Default|GetAttachedPickup))
  (Utilities|FlowControl|Sequence
    (:then_0
      (Transformation|DetachFromActor _attachedpickup "KeepWorld" "KeepWorld" "KeepWorld")
      (Transformation|SetActorLocation _attachedpickup NewLocation false true)
      (Game|Components|ProjectileMovement|StopSimulating (Variables|/Script/FortniteGame.FortPickup|GetMovementComponent _attachedpickup) (Collision|MakeHitResult false false 0.0 0.0 (Transformation|GetActorLocation _attachedpickup) "0, 0, 0" "0,0,1" "0,0,1" 0 (Variables|Default|GetAttachedPickup)))
      (bind _asfort_player_pawn_athena (Utilities|Casting|CastToFortPlayerPawnAthena (Actor|GetOwner)))
      (bind _asfort_player_controller_athena (Utilities|Casting|CastToFortPlayerControllerAthena (Pawn|GetController _asfort_player_pawn_athena)))
      (Fortnite|RegisterSoundIndicatorPlacementContext _asfort_player_controller_athena (Variables|Default|GetAttachedPickup)))
    (:then_1
      (Rendering|SetActorHiddenInGame true)
      (Variables|Default|SetThrown false)
      (Actor|DestroyActor))))

(event EventTick
  (params (DeltaSeconds :type Float%20%28single-precision%29)))

(event OnComponentHit(ParentMesh)
  (params (HitComponent :type Primitive%20Component%20Object%20Reference) (OtherActor :type Actor%20Object%20Reference) (OtherComp :type Primitive%20Component%20Object%20Reference) (NormalImpulse :type Vector :default "0, 0, 0") (Hit :type Hit%20Result%20Structure%20%28by%20ref%29))
  (bind (_bblockinghit _binitialoverlap _time _distance _location _impactpoint _normal) (Collision|BreakHitResult Hit))
  (if (Variables|Default|GetThrown)
    (Networking|SwitchHasAuthority
      (:Authority
        (CallFunction|Detach (+ _location (* _normal 5e+01)))))))

(event Custom|Throw
  (Networking|SwitchHasAuthority
    (:Authority
      (Variables|Default|SetThrown true))))

(fn AttachPickup
  (params (AttachedPickup :type Fort%20Pickup%20Object%20Reference))
  (bind _output_get (Variables|Default|SetAttachedPickup AttachedPickup))
  (Game|Components|ProjectileMovement|StopSimulating (Variables|/Script/FortniteGame.FortPickup|GetMovementComponent _output_get) (Collision|MakeHitResult false)))

(fn OnRep_Thrown
  (Utilities|Array|RemoveItem (Variables|/Script/Engine.Actor|GetTags (Variables|Default|GetAttachedPickup)) "GrabbedByGravyGoblin"))

(fn ConstructionScript
  (|CallParentFunction))
