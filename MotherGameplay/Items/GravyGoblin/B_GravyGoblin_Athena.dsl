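; GravyGoblin (Grabitron) gravity-gun weapon Blueprint for grabbing, holding and throwing objects.
; Targeting filters hits through GravityGunTargetTracking, picks the one nearest camera aim by dot product, then async-traces for obstructions, retrying remaining candidates when blocked.
; Holding state swaps grabbing/holding GameplayEffects and toggles HUD key prompts, while OnRep_HasRing switches ring, charge and beam visuals.
; FindTargettedActor_DEPRECATED is an obsolete trace path superseded by the tracking component.
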(fn FindTargettedActor_DEPRECATED
  (bind _instigator (Variables|Actor|GetInstigator))
  (bind _returnvalue (Fortnite|GetFortPlayerControllerfromActor _instigator))
  (bind _playercameramanager (Variables|/Script/Engine.PlayerController|GetPlayerCameraManager _returnvalue))
  (bind _returnvalue_2 (Camera|GetCameraLocation _playercameramanager))
  (bind _returnvalue_3 (Camera|GetCameraRotation _playercameramanager))
  (bind _returnvalue_4 (Math|Vector|GetForwardVector _returnvalue_3))
  (bind _foundactors (LocalVariables|GetFoundActors))
  ;[1 1 1] DO NOT USE - Use FindTargetActor on the GravityGunTargetTracking component instead
  (bind _outhits (Collision|MultiLineTraceByProfile (Variables|Default|GetTraceStart) (Variables|Default|GetTraceEnd) "FortProjectileBall" false (Utilities|Array|MakeArray (Variables|Actor|GetInstigator)) "None" true "(R=1.000000,G=0.000000,B=0.000000,A=1.000000)" "(R=0.000000,G=1.000000,B=0.000000,A=1.000000)" 0.5))
  (if (> (Utilities|Array|Length _outhits) 0)
    (bind _array_element (Utilities|Array|ForEachLoopwithBreak _outhits :!label _label)
      (:LoopBody
        (bind _result (Class|GravyGoblinFunctionLibrary|CanbePickedUpbyGrabitronDEPRECATED (Variables|Actor|GetInstigator) (.HitActor (Collision|BreakHitResult _array_element))))
        (if _result
          (Utilities|Array|Add (LocalVariables|GetFoundActors) (.HitActor (Collision|BreakHitResult _array_element)))
          (Utilities|Casting|CastToFortPickup (.HitActor (Collision|BreakHitResult _array_element))
            (:CastFailed
              (goto _label Break)))))
      (:Completed
        (switch int (Utilities|Array|Length _foundactors)
          (:Default
            (for _array_element_2 _foundactors
              (if (> (Math|Vector|DotProduct (Math|Vector|Normalize (- (Collision|GetActorBounds _array_element_2 true false) _returnvalue_2)) _returnvalue_4) (LocalVariables|GetHighestDot))
                (LocalVariables|SetHighestDot (Math|Vector|DotProduct (Math|Vector|Normalize (- (Collision|GetActorBounds _array_element_2 true false) _returnvalue_2)) _returnvalue_4))
                (Variables|Default|SetTargetedActor _array_element_2)))
            (CallFunction|FoundTargettedActor (Variables|Default|GetTargetedActor)))
          (:0
            (CallFunction|DeactivateReticle))
          (:1
            (CallFunction|FoundTargettedActor (Utilities|Array|Get(acopy) _foundactors))))))
    (else
      (CallFunction|DeactivateReticle))))

(fn OnRep_HasRing
  (if (Variables|Default|GetHasRing)
    (Development|SetHiddeninGame (Variables|Default|GetRing) false true)
    (Variables|Default|SetShowBeam false)
    (Components|Activation|SetActive (Variables|Default|GetNS_GravyGoblin_Beam1) false false)
    (Components|Activation|Activate (Variables|Default|GetNS_GravyGoblin_Charge) false)
    (else
      (Development|SetHiddeninGame (Variables|Default|GetRing) true true)
      (Variables|Default|SetShowBeam true)
      (Components|Activation|Activate (Variables|Default|GetNS_GravyGoblin_Beam1) false)
      (Components|Activation|SetActive (Variables|Default|GetNS_GravyGoblin_Charge) false false)
      (Effects|Components|ParticleSystem|SetFloatParameter (Variables|Default|GetNS_GravyGoblin_Beam1) "OpacityMult" 1.0)
      (Components|Activation|Activate (Variables|Default|GetNS_GravyGoblin_MuzzleFlash) false)
      (Effects|Components|ParticleSystem|SetFloatParameter (Variables|Default|GetNS_GravyGoblin_Beam1) "InFlight" 1.0))))

(fn RaycastResult
  (params (Succeeded :type Boolean) (ActorHit :type Actor%20Object%20Reference) (HitResult :type Hit%20Result%20Structure))
  (if (Collision|BreakHitResult HitResult)
    (Utilities|Array|RemoveItem (Variables|Default|GetPossibleTargets) (Variables|Default|GetTargetedActor))
    (CallFunction|SetTargettedActor (Utilities|Array|MakeArray) true)
    (else
      (CallFunction|FoundTargettedActor 0))))

(fn ResetGunMode
  (bind _instigator (Variables|Actor|GetInstigator))
  (bind _returnvalue (Fortnite|GetFortPlayerControllerfromActor _instigator))
  (bind _returnvalue_2 (Ability|GetAbilitySystemComponent _returnvalue))
  (Utilities|IsValid _instigator
    (:Is%20Valid
      (Utilities|IsValid _returnvalue_2
        (:Is%20Valid
          (GameplayEffects|RemoveActiveEffectswithGrantedTags _returnvalue_2 "(GameplayTags=((TagName=\"Granted.Athena.GravyGoblin.GrabbingObjectMode\"),(TagName=\"Granted.Athena.GravyGoblin.HoldingObjectMode\")))"))))))

(fn SetHoldingState
  (params (IsHolding :type Boolean))
  (bind _output_get (Variables|Default|SetIsHoldingObject IsHolding))
  (Default|UpdateGunMode false)
  (if _output_get
    (WeaponHUDData|SetHudKeyActionVisibilitybyID "ThrowObject" true)
    (WeaponHUDData|SetHudKeyActionVisibilitybyID "RotateObject" true)
    (WeaponHUDData|SetHudKeyActionVisibilitybyID "Aim" true)
    (else
      (WeaponHUDData|SetHudKeyActionVisibilitybyID "ThrowObject")
      (WeaponHUDData|SetHudKeyActionVisibilitybyID "RotateObject")
      (WeaponHUDData|SetHudKeyActionVisibilitybyID "Aim"))))

(fn SetTargettedActor
  (params (Array :type Array%20of%20Hit%20Result%20Structures) (SkipFillingActorArray :type Boolean))
  (bind _instigator (Variables|Actor|GetInstigator))
  (bind _returnvalue (Fortnite|GetFortPlayerControllerfromActor _instigator))
  (bind _playercameramanager (Variables|/Script/Engine.PlayerController|GetPlayerCameraManager _returnvalue))
  (bind _returnvalue_2 (Camera|GetCameraLocation _playercameramanager))
  (bind _returnvalue_3 (Camera|GetCameraRotation _playercameramanager))
  (bind _returnvalue_4 (Math|Vector|GetForwardVector _returnvalue_3))
  (if (not SkipFillingActorArray)
    (Utilities|Array|Clear (Variables|Default|GetPossibleTargets))
    (bind _array_element (Utilities|Array|ForEachLoopwithBreak Array)
      (:LoopBody
        (bind (_reasonforfailure _returnvalue_5) (Targeting|IsValidTarget (Variables|Default|GetGravityGunTargetTracking) (Variables|Actor|GetInstigator) (.HitActor (Collision|BreakHitResult _array_element)) :!class /Script/MotherGameplayRuntime.GravityGunTargetTrackingComponent))
        (if _returnvalue_5
          (Utilities|Array|Add (Variables|Default|GetPossibleTargets) (.HitActor (Collision|BreakHitResult _array_element)))))
      (:Completed
        _)))
  (switch int (Utilities|Array|Length (Variables|Default|GetPossibleTargets))
    (:Default
      _)
    (:0
      (Variables|Default|SetTargetedActor 0)
      (CallFunction|FoundTargettedActor 0))
    (:1
      (Variables|Default|SetTargetedActor (Utilities|Array|Get(acopy) (Variables|Default|GetPossibleTargets)))
      (CallFunction|FoundTargettedActor 0))
    (:2
      _))
  (for _array_element_2 (Variables|Default|GetPossibleTargets)
    (if (> (Math|Vector|DotProduct (Math|Vector|Normalize (- (Collision|GetActorBounds _array_element_2 true false) _returnvalue_2)) _returnvalue_4) (LocalVariables|GetHighestDot))
      (LocalVariables|SetHighestDot (Math|Vector|DotProduct (Math|Vector|Normalize (- (Collision|GetActorBounds _array_element_2 true false) _returnvalue_2)) _returnvalue_4))
      (Variables|Default|SetTargetedActor _array_element_2)))
  (Class|AsyncTraceQueue|AsyncTraceImmediate (Transformation|GetActorLocation) (Collision|GetActorBounds (Variables|Default|GetTargetedActor) false false) (Utilities|Array|MakeArray self (Variables|Actor|GetInstigator) (Variables|Default|GetTargetedActor)) "ECC_GameTraceChannel7" (EventDispatchers|CreateEvent :!event RaycastResult)))

(fn UpdateGunMode
  (params (IsEquipping :type Boolean))
  (bind _instigator (Variables|Actor|GetInstigator))
  (bind _returnvalue (Fortnite|GetFortPlayerControllerfromActor _instigator))
  (bind _returnvalue_2 (Ability|GetAbilitySystemComponent _returnvalue))
  (if (or IsEquipping (FortWeapon|IsEquipped))
    (Utilities|IsValid _instigator
      (:Is%20Valid
        (Utilities|IsValid _returnvalue_2
          (:Is%20Valid
            (if (Variables|Default|GetIsHoldingObject)
              (GameplayEffects|RemoveActiveEffectswithGrantedTags _returnvalue_2 "(GameplayTags=((TagName=\"Granted.Athena.GravyGoblin.GrabbingObjectMode\")))")
              (GameplayEffects|ApplyGameplayEffectToTarget :self _returnvalue_2 :GameplayEffectClass "/MotherGameplay/Items/GravyGoblin/GE_HoldingObjectMode_GravyGoblin.GE_HoldingObjectMode_GravyGoblin_C" :Target _returnvalue_2 :Level 0.0)
              (else
                (GameplayEffects|RemoveActiveEffectswithGrantedTags _returnvalue_2 "(GameplayTags=((TagName=\"Granted.Athena.GravyGoblin.HoldingObjectMode\")))")
                (GameplayEffects|ApplyGameplayEffectToTarget :self _returnvalue_2 :GameplayEffectClass "/MotherGameplay/Items/GravyGoblin/GE_GrabbingObjectMode_GravyGoblin.GE_GrabbingObjectMode_GravyGoblin_C" :Target _returnvalue_2 :Level 0.0)))))))))

(fn ConstructionScript
  (|CallParentFunction))
