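# Niagara beam effect for the Gravy Goblin item, drawing a purple-to-green ribbon between the system origin and User.SplineUpdatedEndPos.
# Spline places partitioned ribbon particles on a Hermite curve, which both spark emitters also sample for falling or travelling sparks.
# Glow adds sprites and lights near the origin, while Lines sprays streaks from two mirrored sides.
# User.ADS retunes the look, and User.InFlight blends the ribbon tangent and suppresses spark spawning.
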NiagaraSystem /MotherGameplay/Items/GravyGoblin/Effects/NS_GravyGoblin_Beam
dependencies {
 usedDataInterfaces [&NiagaraDataInterfaceCurve]
 usedDynamicInputs [
  &RotateVectorByQuaternion
  &Add_Float
  &Sine
  &Cosine
 ]
 usedModules [
  &SystemState
  &self:SystemSpawnScript.SystemScriptSource.SystemScriptGraph.NiagaraNodeAssignment_1.SetVariables_#1
  &self:SystemSpawnScript.SystemScriptSource.SystemScriptGraph.NiagaraNodeAssignment_0.SetVariables_#2
 ]
}
emitters {
 Glow {
  emitterUpdateScript [
   &EmitterState()
   &SpawnRate(SpawnRate=8, Use Spawn Probability=false)
  ]
  particleSpawnScript [
   &InitializeParticle(
    Lifetime Max=0.4
    Lifetime Min=0.2
    Lifetime Mode=::NewEnumerator1
    Sprite Rotation Mode=::NewEnumerator2
    Sprite Size Mode=::NewEnumerator1
    Uniform Sprite Size Max=70
    Uniform Sprite Size Min=35
    UsePositionOffset=false
    Write Material Random=true
   )
   &BoxLocation(
    Box Offset=<0 10 0>
    Box Size=&Lerp_Vector(A=<12 0 0>, B=<6 0 0>, Alpha=$User.ADS)
    MaskSpawnGroup=false
    Module.Override Seed=false
    Rotate Box=false
    Surface Only=false
   )
   &DynamicMaterialParameters(
    Index 0 Param 1=0.5
    Index 1 Param 1=20
    Module.Param0WriteEnabled=true
    Module.Param0WriteEnabled1=true
    Module.Param1WriteEnabled=true
    Module.Param1WriteEnabled1=true
    Module.Param2WriteEnabled=true
    Module.Param2WriteEnabled1=true
    Module.Param3WriteEnabled=true
    Module.Param3WriteEnabled1=true
    Write Parameter Index 1=true
   )
  ]
  particleUpdateScript [
   &ParticleState()
   &SolveForcesAndVelocity(Clamp Velocity=false, Limit Acceleration=false)
   &CameraOffset(Camera Offset Amount=&Lerp_Float(A=15, B=1, Alpha=$User.ADS))
   &FN_Light_Attributes(
    Color=<rgba 0 1 0.2061 1>
    Create Color Attribute=true
    Create Position Attribute=true
    Create Radius Attribute=true
    Position=&MakeVector(X=0, Y=0, Z=&Multiply_Float(A=2, B=1))
    Radius=&Multiply_Float(A=5, B=1)
    Scale Color=&Lerp_Float(A=1, B=2, Alpha=$User.ADS)
   )
   &ScaleColor(
    Scale Alpha=&FloatFromCurve(
     FloatCurve=DataInterface() {propertyValues {Curve {keys [
         <at 0 0 RCIM_Cubic>
         {interpMode=RCIM_Cubic, tangentMode=RCTM_Break, time=0.3, value=1}
         {interpMode=RCIM_Cubic, tangentMode=RCTM_Break, time=1, value=0}
        ]}}}
     Scale Curve=&Lerp_Float(A=1, B=0.35, Alpha=$User.ADS)
    )
    Scale RGB=&MakeVectorFromLinearColorRGB(Color=&LerpLinearColors(
      StartColor=&MultiplyLinearColor(
       Linear Color=<rgba 0.5292 0.0834 0.75 1>
       Scale Factor=&MakeFloat4FromVectorAndFloat(
        "Vector (RGB)"=&VectorFromFloat(Value=&FloatFromCurve(
          FloatCurve=DataInterface() {propertyValues {Curve {keys [<at 0.6 3 RCIM_Cubic>, {arriveTangent=134.7999, interpMode=RCIM_Cubic, tangentMode=RCTM_Break, time=1, value=20}]}}}
          CurveIndex=&RandomRangeFloat(Minimum=0, Maximum=1)
         ))
        "Float (Alpha)"=1
       )
      )
      EndColor=&MultiplyLinearColor(
       Linear Color=<rgba 0 1 0.2061 1>
       Scale Factor=&MakeFloat4FromVectorAndFloat("Vector (RGB)"=&VectorFromFloat(Value=3), "Float (Alpha)"=1)
      )
      LerpFactor=&FloatFromCurve(FloatCurve=DataInterface() {propertyValues {Curve {keys [
           {arriveTangent=-0.0498, interpMode=RCIM_Cubic, leaveTangent=-0.0498, tangentMode=RCTM_User, time=0.2, value=0}
           {arriveTangent=3.0358, interpMode=RCIM_Cubic, tangentMode=RCTM_Break, time=1, value=1}
          ]}}})
     ))
    ScaleA=true
    ScaleRGB=true
   )
   &ScaleSpriteSize(Scale Factor=&Vector2DFromFloat(Value=&FloatFromCurve(
      FloatCurve=DataInterface() {propertyValues {Curve {keys [
          <at 0 0.5 RCIM_Cubic>
          {interpMode=RCIM_Cubic, tangentMode=RCTM_Break, time=0.3, value=1}
          {interpMode=RCIM_Cubic, tangentMode=RCTM_Break, time=1, value=0.5}
         ]}}}
      Scale Curve=0.65
     )))
   &SpriteRotationRate(Rotation Rate=&RandomRangeFloat(Minimum=-10, Maximum=10))
  ]
  rendererClasses [&NiagaraSpriteRendererProperties, &NiagaraLightRendererProperties]
  renderers [
   {rendererClass=&NiagaraSpriteRendererProperties}
   {rendererClass=&NiagaraLightRendererProperties, rendererIndex=1}
  ]
 }
 Lines {
  emitterUpdateScript [
   &EmitterState()
   /SetVariables(Emitter.RandRotBox:NiagaraFloat=&UniformAOrBFloat(A=-0.2, B=0.2)) {isSetParameters=true}
   &SpawnRate(SpawnRate=20, Use Spawn Probability=false) {enabled=false}
   /Game/Effects/Niagara/Modules/Spawn/SpawnRateFromMultipleSpawnGroups_2(Global Spawnrate=10)
  ]
  particleSpawnScript [
   &FN_SetSpawnGroup(Number of Spawn Groups=2)
   &InitializeParticle(
    Color Maximum=<rgba 0.75 0.0907 0.6327 1>
    Color Minimum=<rgba 0.75 0.0699 0.0923 1>
    Color Mode=::NewEnumerator2
    Lifetime Max=0.25
    Lifetime Min=0.1
    Lifetime Mode=::NewEnumerator1
    Sprite Size=&MakeVector2D(X=0.75, Y=&RandomRangeFloat(Minimum=15, Maximum=20))
    Sprite Size Mode=::NewEnumerator4
    UsePositionOffset=false
    Write Material Random=true
   )
   &CameraOffset(Camera Offset Amount=&Lerp_Float(A=25, B=1, Alpha=$User.ADS))
   &BoxLocation(
    Box Offset=&MakeVector(
     X=/Niagara/DynamicInputs/Math/LerpMultipleFloats(
      Float Count:ENiagara_MultipleLerpCount=::NewEnumerator3
      1:NiagaraFloat=-10
      2:NiagaraFloat=10
      3:NiagaraFloat=?
      4:NiagaraFloat=?
      5:NiagaraFloat=?
      Alpha:NiagaraFloat=&MakeFloatFromInt(Int=$Particles.SpawnGroup)
     )
     Y=8
     Z=0
    )
    Box Size=<1 1 1>
    MaskSpawnGroup=false
    Module.Override Seed=false
    Rotate Box=false
    Surface Only=false
   )
   &AddVelocityInCone(
    Cone Angle=50
    Cone Axis=&MakeVector(
     X=&LerpMultipleFloats(
      Float Count:ENiagara_MultipleLerpCount=::NewEnumerator3
      1:NiagaraFloat=0.75
      2:NiagaraFloat=-0.75
      3:NiagaraFloat=?
      4:NiagaraFloat=?
      5:NiagaraFloat=?
      Alpha:NiagaraFloat=&MakeFloatFromInt(Int=$Particles.SpawnGroup)
     )
     Y=1
     Z=0
    )
    Module.Use Velocity Falloff On Cone Axis=true
    Velocity Distribution Along Cone Axis=1
    Velocity Falloff Away From Cone Axis=1
    Velocity Strength=20
   )
  ]
  particleUpdateScript [
   &ParticleState()
   &SolveForcesAndVelocity(Clamp Velocity=false, Limit Acceleration=false)
   &ScaleSpriteSize(Scale Factor=&MakeVector2D(
     X=&FloatFromCurve(FloatCurve=DataInterface() {propertyValues {Curve {keys [<at 0 1 RCIM_Cubic>, {arriveTangent=9.5356, interpMode=RCIM_Cubic, tangentMode=RCTM_Break, time=1, value=4}]}}})
     Y=&FloatFromCurve(FloatCurve=DataInterface() {propertyValues {Curve {keys [{interpMode=RCIM_Cubic, tangentMode=RCTM_Break, time=0, value=0.2}, {interpMode=RCIM_Cubic, tangentMode=RCTM_Break, time=1, value=1}]}}})
    ))
   &ScaleColor(
    Scale Alpha=&FloatFromCurve(
     FloatCurve=DataInterface() {propertyValues {Curve {keys [
         <at 0 0 RCIM_Cubic>
         {arriveTangent=-0.0008, interpMode=RCIM_Cubic, leaveTangent=-0.0008, tangentMode=RCTM_User, time=0.8, value=1}
         {arriveTangent=-2.2743, interpMode=RCIM_Cubic, tangentMode=RCTM_Break, time=1, value=0.6}
        ]}}}
     Scale Curve=&Lerp_Float(A=1, B=0.35, Alpha=$User.ADS)
    )
    Scale RGB=&VectorFromFloat(Value=&RandomRangeFloat(Minimum=0.5, Maximum=3))
    ScaleA=true
    ScaleRGB=true
   )
   &DynamicMaterialParameters(
    Index 0 Param 1=150
    Module.Param0WriteEnabled=true
    Module.Param1WriteEnabled=true
    Module.Param2WriteEnabled=true
    Module.Param3WriteEnabled=true
   )
  ]
  rendererClasses [&NiagaraSpriteRendererProperties]
  renderers [{rendererClass=&NiagaraSpriteRendererProperties}]
 }
 SparksDrop {
  emitterUpdateScript [
   &EmitterState()
   &SpawnRate(
    Spawn Probability=&Lerp_Float(A=1, B=0, Alpha=$User.InFlight)
    SpawnRate=&Lerp_Float(A=6, B=15, Alpha=$User.ADS)
    Use Spawn Probability=true
   )
  ]
  particleSpawnScript [
   &InitializeParticle(
    Color Maximum=<rgba 0.9115 0.3373 0.0123 1>
    Color Minimum=<rgba 0 1 0.2061 1>
    Color Mode=::NewEnumerator2
    Lifetime Max=0.3
    Lifetime Min=0.25
    Lifetime Mode=::NewEnumerator1
    Sprite Size Max=<3 3>
    Sprite Size Min=<2 2>
    Sprite Size Mode=::NewEnumerator2
    UsePositionOffset=false
    Write Material Random=true
   )
   /Game/Effects/Niagara/Modules/Curves/FN_HermiteCubicSpline_v1/FN_HermiteCubicSpline(
    Direction End:Vector3f=$System.HermiteDirectionStart
    Direction Start:Vector3f=$System.HermiteDirectionEnd
    Position Blend:NiagaraFloat=&RandomRangeFloat(Minimum=0.7, Maximum=0.9)
    Position End:Vector3f=$User.SplineUpdatedEndPos
    Position Start:Vector3f=/Niagara/DynamicInputs/Helpers/SimulationPosition()
    Write Particles Position=true
   )
  ]
  particleUpdateScript [
   &ParticleState()
   /SetVariables(Particles.BeamDirection:Vector3f=&NormalizeVector(Vector To Normalize=&Subtract_Vector(A=$User.SplineUpdatedEndPos, B=&SimulationPosition()))) {isSetParameters=true}
   &ScaleColor(
    Scale Alpha=&Multiply_Float(
     A=&FloatFromCurve(
      FloatCurve=DataInterface() {propertyValues {Curve {keys [{arriveTangent=9.4231, interpMode=RCIM_Cubic, time=0, value=1}, {interpMode=RCIM_Cubic, leaveTangent=-0.779, time=1, value=0}]}}}
      Scale Curve=&RandomRangeFloat(Minimum=0.1, Maximum=0.3)
     )
     B=&FloatFromCurve(
      FloatCurve=DataInterface() {propertyValues {Curve {keys [{arriveTangent=1.8177, interpMode=RCIM_Cubic, leaveTangent=1.8177, tangentMode=RCTM_User, time=0, value=0}, <at 1 1 RCIM_Cubic>]}}}
      CurveIndex=$System.Age
      Scale Curve=&Multiply_Float(A=$User.RingRotX, B=$User.OpacityMult)
     )
    )
    ScaleA=true
    ScaleRGB=false
   )
   &CurlNoiseForce(Noise Strength=500, Pan Noise=false)
   &GravityForce(Gravity=&Multiply_VectorByFloat(Vector=$Particles.BeamDirection, Float=-1000))
   &GravityForce(Gravity=<0 0 -700>) {module=GravityForce001}
   &SolveForcesAndVelocity(Clamp Velocity=false, Limit Acceleration=false)
   &ScaleSpriteSize(Scale Factor=&Vector2DFromFloat(Value=&Lerp_Float(A=1, B=0.7, Alpha=$User.ADS)))
  ]
  rendererClasses [&NiagaraSpriteRendererProperties]
  renderers [{rendererClass=&NiagaraSpriteRendererProperties}]
 }
 SparksHermite_End {
  emitterUpdateScript [
   &EmitterState(
    Inactive Response=::NewEnumerator1
    Life Cycle Mode=::NewEnumerator1
    Loop Behavior=::NewEnumerator1
    Loop Duration Mode=::NewEnumerator1
    UseLoopDelay=false
   )
   &SpawnRate(
    Spawn Probability=&Lerp_Float(A=1, B=0, Alpha=$User.InFlight)
    SpawnRate=&Lerp_Float(A=6, B=12, Alpha=$User.ADS)
    Use Spawn Probability=true
   )
  ]
  particleSpawnScript [&InitializeParticle(
    Lifetime Max=1.5
    Lifetime Min=0.5
    Lifetime Mode=::NewEnumerator1
    Sprite Size Max=<3.5 3.5>
    Sprite Size Min=<2 2>
    Sprite Size Mode=::NewEnumerator2
    UsePositionOffset=false
    Write Material Random=true
   )]
  particleUpdateScript [
   &ParticleState()
   &FN_HermiteCubicSpline(
    Compute Direction=true
    Direction End:Vector3f=$System.HermiteDirectionStart
    Direction Start:Vector3f=$System.HermiteDirectionEnd
    Position Blend:NiagaraFloat=/Game/Effects/Niagara/DynamicInputs/Math/FN_Fit01(Clamp:NiagaraBool=false, Value:NiagaraFloat=$Particles.NormalizedAge, DestMin:NiagaraFloat=0.3, DestMax:NiagaraFloat=1)
    Position End:Vector3f=$User.SplineUpdatedEndPos
    Position Start:Vector3f=&SimulationPosition()
    Write Particles Position=true
    Write Particles Velocity=true
   ) {module=FN_HermiteCubicSpline001}
   &ScaleSpriteSize(Scale Factor=&Vector2DFromFloat(Value=&FloatFromCurve(
      FloatCurve=DataInterface() {propertyValues {Curve {keys [
          {arriveTangent=0.8246, interpMode=RCIM_Cubic, leaveTangent=0.8246, tangentMode=RCTM_User, time=0, value=0.4}
          {arriveTangent=-0.0136, interpMode=RCIM_Cubic, leaveTangent=-0.0136, tangentMode=RCTM_User, time=1, value=1}
         ]}}}
      Scale Curve=&Lerp_Float(A=1, B=2, Alpha=$User.ADS)
     )))
   &ScaleColor(
    Scale Alpha=&Multiply_Float(
     A=&FloatFromCurve(
      FloatCurve=DataInterface() {propertyValues {Curve {keys [
          {arriveTangent=9.4231, interpMode=RCIM_Cubic, leaveTangent=4.0867, tangentMode=RCTM_User, time=0, value=0}
          <at 0.5 1 RCIM_Cubic>
          {arriveTangent=0.1025, interpMode=RCIM_Cubic, leaveTangent=-0.0292, tangentMode=RCTM_User, time=1, value=0}
         ]}}}
      Scale Curve=0.05
     )
     B=&Multiply_Float(A=1, B=$User.OpacityMult)
    )
    Scale RGB=&MakeVectorFromLinearColorRGB(Color=&LerpLinearColors(
      StartColor=<rgba 0 1 0.2061 1>
      EndColor=<rgba 0.1876 0.0825 0.75 1>
      LerpFactor=&FloatFromCurve(FloatCurve=DataInterface() {propertyValues {Curve {keys [{interpMode=RCIM_Cubic, tangentMode=RCTM_Break, time=0, value=1}, {interpMode=RCIM_Cubic, tangentMode=RCTM_Break, time=1, value=0}]}}})
     ))
    ScaleA=true
    ScaleRGB=true
   )
  ]
  rendererClasses [&NiagaraSpriteRendererProperties]
  renderers [{rendererClass=&NiagaraSpriteRendererProperties}]
 }
 Spline {
  emitterSpawnScript [/Game/Effects/Niagara/Modules/Partition/FN_PartitionAttributes(Number of Bone Sets=1, Particles Per Partition=5)]
  emitterUpdateScript [
   &EmitterState(
    Inactive Response=::NewEnumerator1
    Life Cycle Mode=::NewEnumerator1
    Loop Behavior=::NewEnumerator1
    Loop Duration Mode=::NewEnumerator1
    UseLoopDelay=false
   )
   &SpawnBurst_Instantaneous(Spawn Count=$Emitter.FN_PartitionAttributes.Spawn Total, Use Spawn Probability=false)
  ]
  particleSpawnScript [
   /Niagara/Modules/Spawn/Initialization/InitializeParticle(
    Lifetime=5
    Module.Write Color=true
    Module.Write Lifetime=true
    Module.Write Mass=true
    Module.Write Material Random=true
    Module.Write Position=true
    Module.Write Scale=false
    Module.Write SpriteRotation=false
    Module.Write SpriteSize=true
    Sprite Size=<25 25>
   )
   /Game/Effects/Niagara/Modules/Partition/FN_Particle_partition_v5/FN_Particle_Partition_v5(Enable Ribbon ID=true)
   &InitializeRibbon(
    Lifetime=0.7
    Module.Write Color=true
    Module.Write Lifetime=true
    Module.Write Mass=true
    Module.Write Material Random=true
    Module.Write Position=false
    Module.Write RibbonTwist=false
    Module.Write RibbonWidth=true
    Ribbon Width=30
   )
  ]
  particleUpdateScript [
   &ParticleState(Kill Particles When Lifetime Has Elapsed=false, Let Infinitely Lived Particles Die When Emitter Deactivates=true)
   /SetVariables() {isSetParameters=true}
   &SystemLocation(MaskSpawnGroup=false)
   &FN_HermiteCubicSpline(
    Direction End:Vector3f=&Lerp_Vector(A=$System.HermiteDirectionEnd, B=$System.HermiteDirectionEndAlt, Alpha=$User.InFlight)
    Direction Start:Vector3f=$System.HermiteDirectionStart
    Position Blend:NiagaraFloat=$Particles.Initial.Partition.NormalizedID
    Position End:Vector3f=&SimulationPosition()
    Position Start:Vector3f=$User.SplineUpdatedEndPos
    Write Particles Position=true
   )
   &FN_RibbonMaterial(
    Color=&LerpLinearColors(
     StartColor=<rgba 0.5292 0.0834 0.75 1>
     EndColor=<rgba 0 1 0.2061 1>
     LerpFactor=&FloatFromCurve(
      FloatCurve=DataInterface() {propertyValues {Curve {keys [{interpMode=RCIM_Cubic, tangentMode=RCTM_Break, time=0, value=1}, {interpMode=RCIM_Cubic, tangentMode=RCTM_Break, time=1, value=0}]}}}
      CurveIndex=$Particles.Partition.NormalizedID
     )
    )
    Color Pan U Speed=-2
    Erosion Amount=1
    Erosion Pan Speed U=-2
    Erossion Scale U=4
    Macro Pan Speed=-4
    Micro Distortion Amount=0.25
    Micro Pan Speed=-1
    Micro Scale U=0.25
    Opacity Erosion:NiagaraFloat=&RandomRangeFloat(Minimum=0, Maximum=0)
    Opacity Pan U Speed=-2
    Opacity Scale U=0.2
    Scale Alpha:NiagaraFloat=&FloatFromCurve(
     FloatCurve=DataInterface() {propertyValues {Curve {keys [{arriveTangent=2.1027, interpMode=RCIM_Cubic, leaveTangent=2.1027, tangentMode=RCTM_User, time=0, value=0}, <at 1 1 RCIM_Cubic>]}}}
     Scale Curve=&Multiply_Float(
      A=1
      B=&FloatFromCurve(
       FloatCurve=DataInterface() {propertyValues {Curve {keys [
           {arriveTangent=0.0575, interpMode=RCIM_Cubic, leaveTangent=0.0575, tangentMode=RCTM_User, time=0, value=0.2}
           {interpMode=RCIM_Cubic, tangentMode=RCTM_Break, time=0.2, value=1}
           <at 1 1>
          ]}}}
       CurveIndex=$Particles.Partition.NormalizedID
       Scale Curve=&Multiply_Float(A=1, B=$User.OpacityMult)
      )
     )
    )
    Scale Color:Vector3f=&VectorFromFloat(Value=&FloatFromCurve(
      FloatCurve=DataInterface() {propertyValues {Curve {keys [
          <at 0 10 RCIM_Cubic>
          {interpMode=RCIM_Cubic, tangentMode=RCTM_Break, time=0.1165, value=2}
          {interpMode=RCIM_Cubic, tangentMode=RCTM_Break, time=0.9, value=1}
          {interpMode=RCIM_Cubic, tangentMode=RCTM_Break, time=1, value=7}
         ]}}}
      CurveIndex=$Particles.Partition.NormalizedID
      Scale Curve=&Lerp_Float(A=&Lerp_Float(A=1.5, B=5, Alpha=$User.BeamBrightness), B=2, Alpha=$User.ADS)
     ))
   )
   &ScaleRibbonWidth(Ribbon Width Scale=&Multiply_Float(
     A=&FloatFromCurve(
      FloatCurve=DataInterface() {propertyValues {Curve {keys [
          <at 0 1 RCIM_Cubic>
          {arriveTangent=0.125, interpMode=RCIM_Cubic, leaveTangent=0.125, time=0.8052, value=0.9412}
          {arriveTangent=15.6259, interpMode=RCIM_Cubic, leaveTangent=15.6259, tangentMode=RCTM_User, time=1, value=2}
         ]}}}
      CurveIndex=$Particles.Partition.NormalizedID
      Scale Curve=&Lerp_Float(A=0.2, B=0.3, Alpha=$User.Projectile)
     )
     B=&Lerp_Float(A=1, B=1.2, Alpha=$User.ADS)
    ))
  ]
  rendererClasses [&NiagaraRibbonRendererProperties]
  renderers [{rendererClass=&NiagaraRibbonRendererProperties}]
 }
}
summary {
 systemName=NS_GravyGoblin_Beam
 emitters [
  {emitterName=Spline, rendererClasses [&NiagaraRibbonRendererProperties]}
  {emitterName=SparksHermite_End, rendererClasses [&NiagaraSpriteRendererProperties]}
  {emitterName=SparksDrop, rendererClasses [&NiagaraSpriteRendererProperties]}
  {emitterName=Glow, rendererClasses [&NiagaraSpriteRendererProperties, &NiagaraLightRendererProperties]}
  {emitterName=Lines, rendererClasses [&NiagaraSpriteRendererProperties]}
 ]
 userVariables [
  User.ADS {type=NiagaraFloat, defaultValue=0, description=""}
  User.BeamBrightness {type=NiagaraFloat, defaultValue=0, description=""}
  User.InFlight {type=NiagaraFloat, defaultValue=1, description=""}
  User.OpacityMult {type=NiagaraFloat, defaultValue=1, description=""}
  User.Projectile {type=NiagaraFloat, defaultValue=0, description=""}
  User.RibbonIDOffset {type=NiagaraInt32}
  User.RingAxis {type=Vector3f, description="", defaultValue {struct=Vector3f, value=<0 0 0>}}
  User.RingRotX {type=NiagaraFloat, defaultValue=1, description=""}
  User.RotTest {type=Vector3f, description="", defaultValue {struct=Vector3f, value=<0 0 0>}}
  User.SplineUpdatedEndPos {type=Vector3f, description="", defaultValue {struct=Vector3f, value=<0 0 0>}}
  User.SplineUpdatedEndTangent {type=Vector3f, description="", defaultValue {struct=Vector3f, value=<0 0 0>}}
 ]
}
