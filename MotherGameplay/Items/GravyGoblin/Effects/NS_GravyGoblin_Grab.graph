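# Niagara system for the Gravy Goblin item's grab effect: a brief one-shot green flash scaled by User.RingSize.
# Flare drives two sprite renderers with separate curve-animated size modules, one rotating, camera offsets derived from owner LOD distance, and an animated radial-mask material.
# Glow adds a fading sprite and light.
# Green is hard-coded; User.LightColor and the other user variables besides RingSize are not referenced in the shown modules.
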NiagaraSystem /MotherGameplay/Items/GravyGoblin/Effects/NS_GravyGoblin_Grab
dependencies {
 usedDataInterfaces [&NiagaraDataInterfaceCurve]
 usedDynamicInputs [&MakeLinearColorFromVectorAndFloat]
 usedModules [&SystemState]
}
emitters {
 Flare {
  emitterUpdateScript [
   &EmitterState()
   &SpawnBurst_Instantaneous(Spawn Count=1, Use Spawn Probability=false)
  ]
  particleSpawnScript [&InitializeParticle(
    Lifetime=0.25
    Sprite Rotation Mode=::NewEnumerator2
    Sprite Size Mode=::NewEnumerator3
    Uniform Sprite Size=3000
    UsePositionOffset=false
    Write Material Random=true
   )]
  particleUpdateScript [
   &ParticleState()
   &ScaleColor(
    Scale Alpha=&FloatFromCurve(FloatCurve=DataInterface() {propertyValues {Curve {keys [<at 0 0 RCIM_Cubic>, {arriveTangent=0.8, interpMode=RCIM_Cubic, leaveTangent=0.8, time=0.25, value=1}, <at 0.75 1 RCIM_Cubic>, <at 1 0>]}}})
    Scale RGB=&MakeVectorFromLinearColorRGB(Color=<rgba 0 1 0.2061 1>)
    ScaleA=true
    ScaleRGB=true
   )
   /SetVariables(Particles.Size:NiagaraFloat=&FloatFromCurve(
     FloatCurve=DataInterface() {propertyValues {Curve {keys [
         <at -0.0039 0.0032 RCIM_Cubic>
         {arriveTangent=0.0358, interpMode=RCIM_Cubic, leaveTangent=0.0358, tangentMode=RCTM_User, time=0.4967, value=0.9107}
         <at 1 0>
        ]}}}
     Scale Curve=&FloatFromCurve(
      FloatCurve=DataInterface() {propertyValues {Curve {keys [
          <at -0.0005 0.984 RCIM_Cubic>
          {arriveTangent=0.3225, interpMode=RCIM_Cubic, leaveTangent=0.3225, tangentMode=RCTM_User, time=0.4833, value=3.7073}
          <at 1.0108 1.1471>
         ]}}}
      CurveIndex=$Emitter.NormalizedLoopAge
      Scale Curve=0.5
     )
    )) {isSetParameters=true}
   &FN_Sprite_Attributes(
    Camera Offset:NiagaraFloat=&Multiply_Float(A=0.9, B=$Engine.Owner.LODDistance)
    Create Camera Offset Attribute=true
    Create SpriteSize Attribute=true
    SpriteSize=&Vector2DFromFloat(Value=&FloatFromCurve(
      FloatCurve=DataInterface() {propertyValues {Curve {keys [
          <at 0.01 0.1312 RCIM_Cubic>
          <at 0.2296 1.3423>
          {arriveTangent=-1.9697, interpMode=RCIM_Cubic, leaveTangent=-1.9697, tangentMode=RCTM_User, time=0.5965, value=0.0067}
         ]}}}
      Scale Curve=&Multiply_Float(A=20, B=$User.RingSize)
     ))
    Write to Module Namespace=true
   ) {module=FN_Sprite_Attributes002}
   &FN_Sprite_Attributes(
    Camera Offset:NiagaraFloat=&Multiply_Float(A=0.9, B=$Engine.Owner.LODDistance)
    Create Camera Offset Attribute=true
    Create Sprite Rotation Attribute=true
    Create SpriteSize Attribute=true
    SpriteRotation:NiagaraFloat=&FloatFromCurve(FloatCurve=DataInterface() {propertyValues {Curve {keys [<at 0 1>,<at 1 0>]}}}, Scale Curve=200)
    SpriteSize=&Vector2DFromFloat(Value=&FloatFromCurve(
      FloatCurve=DataInterface() {propertyValues {Curve {keys [{arriveTangent=2.4627, interpMode=RCIM_Cubic, leaveTangent=2.4627, tangentMode=RCTM_User, time=0.4909, value=0.0264}, <at 0.9996 0.9847>]}}}
      Scale Curve=&Multiply_Float(A=30, B=$User.RingSize)
     ))
    Write to Module Namespace=true
   ) {module=FN_Sprite_Attributes005}
   &FN_Radial_Material(
    Base Sprite Offset X:NiagaraFloat=&FloatFromCurve(FloatCurve=DataInterface() {propertyValues {Curve {keys [
         {arriveTangent=-1.3361, interpMode=RCIM_Cubic, leaveTangent=-1.3361, tangentMode=RCTM_User, time=0.6845, value=0.9914}
         {arriveTangent=-1.2617, interpMode=RCIM_Cubic, leaveTangent=-1.2617, tangentMode=RCTM_User, time=0.9897, value=0.6193}
        ]}}})
    Depth Fade=200
    Radial Mask Offset X=-1
    Radial Mask Power:NiagaraFloat=&FloatFromCurve(
     FloatCurve=DataInterface() {propertyValues {Curve {keys [{arriveTangent=6.6825, interpMode=RCIM_Cubic, leaveTangent=6.6825, tangentMode=RCTM_User, time=0.709, value=1.1078}, <at 0.9976 2.0156>]}}}
     Scale Curve=20
    )
    Write to Module Namespace=true
   )
  ]
  rendererClasses [&NiagaraSpriteRendererProperties]
  renderers [
   {rendererClass=&NiagaraSpriteRendererProperties}
   {rendererClass=&NiagaraSpriteRendererProperties, rendererIndex=1}
  ]
 }
 Glow {
  emitterUpdateScript [
   &EmitterState()
   &SpawnBurst_Instantaneous(Spawn Count=1, Use Spawn Probability=false)
  ]
  particleSpawnScript [&InitializeParticle(
    Color=<rgba 0 1 0.2061 1>
    Lifetime=0.25
    Sprite Rotation Mode=::NewEnumerator2
    Sprite Size Mode=::NewEnumerator3
    Uniform Sprite Size=&Multiply_Float(A=75, B=$User.RingSize)
    UsePositionOffset=false
    Write Material Random=true
   )]
  particleUpdateScript [
   &ParticleState()
   &ScaleColor(
    Scale Alpha=&FloatFromCurve(FloatCurve=DataInterface() {propertyValues {Curve {keys [
         {arriveTangent=11.8795, interpMode=RCIM_Cubic, leaveTangent=11.8795, tangentMode=RCTM_User, time=0, value=0}
         {arriveTangent=-0.3146, interpMode=RCIM_Cubic, leaveTangent=-0.3146, tangentMode=RCTM_User, time=0.1309, value=2}
         {arriveTangent=-3.682, interpMode=RCIM_Cubic, leaveTangent=-3.682, tangentMode=RCTM_User, time=0.2098, value=0.9932}
         <at 1 0 RCIM_Cubic>
        ]}}})
    Scale RGB=&VectorFromFloat(Value=2)
    ScaleA=true
    ScaleRGB=true
   )
   &DynamicMaterialParameters(
    Index 0 Param 1=1
    Index 0 Param 2=1
    Index 1 Param 1=400
    Module.Param0WriteEnabled=true
    Module.Param0WriteEnabled1=true
    Module.Param1WriteEnabled=true
    Module.Param1WriteEnabled1=false
    Module.Param2WriteEnabled=false
    Module.Param2WriteEnabled1=false
    Module.Param3WriteEnabled=false
    Module.Param3WriteEnabled1=false
    Write Parameter Index 1=true
   )
   &FN_Light_Attributes(
    Color=<rgba 0 1 0.2061 1>
    Create Color Attribute=true
    Create Exponent Attribute=true
    Create Radius Attribute=true
    Radius=&Multiply_Float(A=70, B=$User.RingSize)
    Scale Color=&FloatFromCurve(
     FloatCurve=DataInterface() {propertyValues {Curve {keys [<at 0 1 RCIM_Cubic>, {arriveTangent=-2.1204, interpMode=RCIM_Cubic, tangentMode=RCTM_Break, time=1, value=0}]}}}
     Scale Curve=2
    )
   )
  ]
  rendererClasses [&NiagaraSpriteRendererProperties, &NiagaraLightRendererProperties]
  renderers [
   {rendererClass=&NiagaraSpriteRendererProperties}
   {rendererClass=&NiagaraLightRendererProperties, rendererIndex=1}
  ]
 }
}
summary {
 systemName=NS_GravyGoblin_Grab
 emitters [
  {emitterName=Flare, rendererClasses [&NiagaraSpriteRendererProperties]}
  {emitterName=Glow, rendererClasses [&NiagaraSpriteRendererProperties, &NiagaraLightRendererProperties]}
 ]
 userVariables [
  User.ADS {type=NiagaraFloat, defaultValue=0, description=""}
  User.AlphaMult {type=NiagaraFloat, defaultValue=0, description=""}
  User.bIsFrontEndPreview {type=NiagaraBool}
  User.ForwardVector {type=Vector3f, description="", defaultValue {struct=Vector3f, value=<1 0 0>}}
  User.LightColor {type=LinearColor, description="", defaultValue {struct=LinearColor, value=<rgba 0 1 0.2061 1>}}
  User.RingSize {type=NiagaraFloat, defaultValue=15, description=""}
 ]
}
