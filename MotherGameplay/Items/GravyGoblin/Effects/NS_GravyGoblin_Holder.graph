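# Niagara effect for the GravyGoblin item's holder: glowing green and purple rings with energy at three fixed prong points.
# Flare, Glow (with lights), GlowCone and Lines cycle a counter modulo three into position arrays, offset via User.NewScale, and fade with User.LaunchBrightness.
# Ring pulses at the center, while RingInX meshes and decal sprites drift both ways along X in two spawn groups.
# User.RingSize scales most elements; User.ADS boosts ring brightness, spawn rate and wave frequency.
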NiagaraSystem /MotherGameplay/Items/GravyGoblin/Effects/NS_GravyGoblin_Holder
dependencies {
 usedDataInterfaces [
  &NiagaraDataInterfaceCurve
  /Script/Niagara.NiagaraDataInterfaceSimpleCounter
  &NiagaraDataInterfaceArrayFloat3
  &NiagaraDataInterfaceColorCurve
  &NiagaraDataInterfaceVectorCurve
 ]
 usedDynamicInputs []
 usedModules [&SystemState]
}
emitters {
 Flare {
  emitterUpdateScript [
   &EmitterState()
   &SpawnBurst_Instantaneous(Spawn Count=3, Use Spawn Probability=false)
  ]
  particleSpawnScript [
   /SetVariables(
    Particles.PlacementPicker:NiagaraInt32=&ModuloInt(
     A=/Niagara/DynamicInputs/DataInterface/GenerateIntFromCounter(SimpleCounter:NiagaraDataInterfaceSimpleCounter=DataInterface() {propertyValues {GpuSyncMode=None, InitialValue=0}})
     B=3
    )
    Particles.SpawnGroup=&ModuloInt(
     A=&GenerateIntFromCounter(SimpleCounter:NiagaraDataInterfaceSimpleCounter=DataInterface() {propertyValues {GpuSyncMode=None, InitialValue=0}})
     B=3
    )
   ) {isSetParameters=true, module=SetVariables_#1}
   /SetVariables(Particles.PositionArray=DataInterface() {propertyValues {GpuSyncMode=SyncCpuToGpu, MaxElements=0, FloatData [<0 0 -7>, <0 6.5 4>, <0 -6.5 4>]}}) {isSetParameters=true, module=SetVariables_#2}
   /SetVariables(Particles.PositionOffset=DataInterface() {propertyValues {GpuSyncMode=SyncCpuToGpu, MaxElements=0, FloatData [<0 0 0>, <0 0.1 8>, <0 0.1 8>]}}) {isSetParameters=true, module=SetVariables_#3}
   &InitializeParticle(
    Color=<rgba 0 1 0.2061 0.25>
    Sprite Size Mode=::NewEnumerator3
    Uniform Sprite Size=&Multiply_Float(A=90, B=1)
    UsePositionOffset=false
    Write Material Random=true
   )
   &SphereLocation(
    MaskSpawnGroup=false
    Offset=&self:PositionOffset(
     ArrayIndex:NiagaraInt32=$Particles.SpawnGroup
     NewScale:NiagaraFloat=$User.NewScale
     PositionArray:NiagaraDataInterfaceArrayFloat3=$Particles.PositionArray
     PositionOffset:NiagaraDataInterfaceArrayFloat3=$Particles.PositionOffset
    )
    Override Seed=false
    Sphere Radius=0
    Surface Only=false
    UseNonUniformScale=false
    UseOffset=true
    UseSphereOrientation=false
   )
  ]
  particleUpdateScript [
   &ParticleState()
   &ScaleColor(Scale Alpha=&Lerp_Float(A=0, B=1, Alpha=$User.LaunchBrightness), ScaleA=true, ScaleRGB=false)
   &ScaleSpriteSize(Scale Factor=&Vector2DFromFloat(Value=&Lerp_Float(A=0.75, B=1, Alpha=$User.LaunchBrightness)))
  ]
  rendererClasses [&NiagaraSpriteRendererProperties]
  renderers [{rendererClass=&NiagaraSpriteRendererProperties}]
 }
 Glow {
  emitterUpdateScript [
   &EmitterState()
   &SpawnBurst_Instantaneous(Spawn Count=3, Use Spawn Probability=false)
  ]
  particleSpawnScript [
   /SetVariables(
    Particles.PlacementPicker:NiagaraInt32=&ModuloInt(
     A=&GenerateIntFromCounter(SimpleCounter:NiagaraDataInterfaceSimpleCounter=DataInterface() {propertyValues {GpuSyncMode=None, InitialValue=0}})
     B=3
    )
    Particles.SpawnGroup=&ModuloInt(
     A=&GenerateIntFromCounter(SimpleCounter:NiagaraDataInterfaceSimpleCounter=DataInterface() {propertyValues {GpuSyncMode=None, InitialValue=0}})
     B=3
    )
   ) {isSetParameters=true, module=SetVariables_#1}
   /SetVariables(Particles.PositionArray=DataInterface() {propertyValues {GpuSyncMode=SyncCpuToGpu, MaxElements=0, FloatData [<0 0 -7>, <0 6.5 4>, <0 -6.5 4>]}}) {isSetParameters=true, module=SetVariables_#2}
   /SetVariables(Particles.PositionOffset=DataInterface() {propertyValues {GpuSyncMode=SyncCpuToGpu, MaxElements=0, FloatData [<0 0 0>, <0 0.1 8>, <0 0.1 8>]}}) {isSetParameters=true, module=SetVariables_#3}
   &InitializeParticle(
    Color=<rgba 0 1 0.2061 1>
    Sprite Size Mode=::NewEnumerator3
    Uniform Sprite Size=&Multiply_Float(A=100, B=1)
    UsePositionOffset=false
    Write Material Random=true
   )
   &SphereLocation(
    MaskSpawnGroup=false
    Offset=&self:PositionOffset(
     ArrayIndex:NiagaraInt32=$Particles.SpawnGroup
     NewScale:NiagaraFloat=$User.NewScale
     PositionArray:NiagaraDataInterfaceArrayFloat3=$Particles.PositionArray
     PositionOffset:NiagaraDataInterfaceArrayFloat3=$Particles.PositionOffset
    )
    Override Seed=false
    Sphere Radius=0
    Surface Only=false
    UseNonUniformScale=false
    UseOffset=true
    UseSphereOrientation=false
   )
   &DynamicMaterialParameters(
    Index 0 Param 1=0.5
    Module.Param0WriteEnabled=true
    Module.Param1WriteEnabled=true
    Module.Param2WriteEnabled=true
    Module.Param3WriteEnabled=true
   )
  ]
  particleUpdateScript [
   &ParticleState()
   &FN_Light_Attributes(
    Color=<rgba 0.1872 1 0.9801 1>
    Create Color Attribute=true
    Create Exponent Attribute=true
    Create Position Attribute=true
    Create Radius Attribute=true
    Position:NiagaraPosition=$Particles.Position
    Radius=&Multiply_Float(A=15, B=$User.RingSize)
    Scale Color=&Lerp_Float(A=2, B=2, Alpha=$User.ADS)
   )
   &ScaleColor(Scale Alpha=&Lerp_Float(A=0, B=1, Alpha=$User.LaunchBrightness), ScaleA=true, ScaleRGB=false)
   &ScaleSpriteSize(Scale Factor=&Vector2DFromFloat(Value=&Lerp_Float(A=0.75, B=1, Alpha=$User.LaunchBrightness)))
  ]
  rendererClasses [&NiagaraSpriteRendererProperties, &NiagaraLightRendererProperties]
  renderers [
   {rendererClass=&NiagaraSpriteRendererProperties}
   {rendererClass=&NiagaraLightRendererProperties, rendererIndex=1}
  ]
 }
 GlowCone {
  emitterUpdateScript [
   &EmitterState()
   &SpawnBurst_Instantaneous(Spawn Count=3, Use Spawn Probability=false) {enabled=false}
   &SpawnRate(SpawnRate=25, Use Spawn Probability=false)
  ]
  particleSpawnScript [
   /SetVariables(
    Particles.PlacementPicker:NiagaraInt32=&ModuloInt(
     A=&GenerateIntFromCounter(SimpleCounter:NiagaraDataInterfaceSimpleCounter=DataInterface() {propertyValues {GpuSyncMode=None, InitialValue=0}})
     B=3
    )
    Particles.SpawnGroup=&ModuloInt(
     A=&GenerateIntFromCounter(SimpleCounter:NiagaraDataInterfaceSimpleCounter=DataInterface() {propertyValues {GpuSyncMode=None, InitialValue=0}})
     B=3
    )
   ) {isSetParameters=true, module=SetVariables_#1}
   /SetVariables(Particles.PositionArray=DataInterface() {propertyValues {GpuSyncMode=SyncCpuToGpu, MaxElements=0, FloatData [<0 0 -7>, <0 6.5 4>, <0 -6.5 4>]}}) {isSetParameters=true, module=SetVariables_#2}
   /SetVariables(Particles.PositionOffset=DataInterface() {propertyValues {GpuSyncMode=SyncCpuToGpu, MaxElements=0, FloatData [<0 0 0>, <0 0.1 8>, <0 0.1 8>]}}) {isSetParameters=true, module=SetVariables_#3}
   /SetVariables(Particles.RotationArray=DataInterface() {propertyValues {GpuSyncMode=SyncCpuToGpu, MaxElements=0, FloatData [<0 -0.25 0>, <0.15 0 -0.25>, <0 0.2 0.25>]}}) {isSetParameters=true, module=SetVariables_#4}
   &InitializeParticle(
    Color=&ColorFromCurve(
     ColorCurve=DataInterface() {propertyValues {
       AlphaCurve {keys [<at 0 1 RCIM_Cubic>,<at 1 1 RCIM_Cubic>]}
       BlueCurve {keys [<at 0 0.2061 RCIM_Cubic>,<at 1 0.75 RCIM_Cubic>]}
       GreenCurve {keys [<at 0 1 RCIM_Cubic>,<at 1 0.0825 RCIM_Cubic>]}
       RedCurve {keys [<at 0 0 RCIM_Cubic>,<at 1 0.1876 RCIM_Cubic>]}
      }}
     CurveIndex=&FloatFromCurve(
      FloatCurve=DataInterface() {propertyValues {Curve {keys [
          {arriveTangent=1.8039, interpMode=RCIM_Cubic, leaveTangent=1.8039, tangentMode=RCTM_User, time=0, value=0}
          {interpMode=RCIM_Cubic, tangentMode=RCTM_Break, time=1, value=1}
         ]}}}
      CurveIndex=&RandomRangeFloat(Minimum=0, Maximum=1)
     )
    )
    Lifetime Max=0.4
    Lifetime Min=0.25
    Lifetime Mode=::NewEnumerator1
    Mesh Scale=&Multiply_VectorByFloat(Vector=&RandomRangeVector(Minimum=<1 1 1>, Maximum=<1.5 1 1>), Float=0.2)
    Mesh Scale Mode=::NewEnumerator4
    UsePositionOffset=false
    Write Material Random=true
   )
   &SphereLocation(
    MaskSpawnGroup=false
    Offset=&self:PositionOffset(
     ArrayIndex:NiagaraInt32=$Particles.SpawnGroup
     NewScale:NiagaraFloat=$User.NewScale
     PositionArray:NiagaraDataInterfaceArrayFloat3=$Particles.PositionArray
     PositionOffset:NiagaraDataInterfaceArrayFloat3=$Particles.PositionOffset
    )
    Override Seed=false
    Sphere Radius=0
    Surface Only=false
    UseNonUniformScale=false
    UseOffset=true
    UseSphereOrientation=false
   )
   &DynamicMaterialParameters(
    Index 0 Param 1=0.5
    Module.Param0WriteEnabled=true
    Module.Param1WriteEnabled=true
    Module.Param2WriteEnabled=true
    Module.Param3WriteEnabled=true
   )
   &InitialMeshOrientation(
    Rotation=&self:RotationOffset_01(
     ArrayIndex:NiagaraInt32=$Particles.SpawnGroup
     RotationArray:NiagaraDataInterfaceArrayFloat3=$Particles.RotationArray
     VecOffset:Vector3f=&RandomRangeVector(Minimum=<-0.04 -0.04 -0.04>, Maximum=<0.04 0.04 0.04>)
    )
    Rotation Coordinate Space=::NewEnumerator0
    Use Orientation Vector=true
    Use Rotation Vector=true
   )
  ]
  particleUpdateScript [
   &ParticleState()
   &ScaleColor(Scale Alpha=&Lerp_Float(A=0, B=0.1, Alpha=$User.LaunchBrightness), ScaleA=true, ScaleRGB=false) {module=ScaleColor001}
   &ScaleColor(
    Scale Alpha=&FloatFromCurve(
     FloatCurve=DataInterface() {propertyValues {Curve {keys [<at 0 0 RCIM_Cubic>,<at 0.5 1.5 RCIM_Cubic>,<at 1 0 RCIM_Cubic>]}}}
     Scale Curve=&Lerp_Float(A=0.7, B=1, Alpha=$User.ADS)
    )
    Scale RGB=&VectorFromFloat(Value=3)
    ScaleA=true
    ScaleRGB=true
   )
   &ScaleMeshSize(Scale Factor=&VectorFromCurve(
     VectorCurve=DataInterface() {propertyValues {
       XCurve {keys [<at 0 0.55>,<at 0.2 1>,<at 1 0.55>]}
       YCurve {keys [<at 0 0.55>,<at 0.2 1>,<at 1 0.55>]}
       ZCurve {keys [<at 0 0.55>,<at 0.2 1>,<at 1 0.55>]}
      }}
     Scale Curve=&VectorFromFloat(Value=$User.RingSize)
    ))
  ]
  rendererClasses [&NiagaraMeshRendererProperties]
  renderers [{rendererClass=&NiagaraMeshRendererProperties, rendererIndex=0}]
 }
 Lines {
  emitterUpdateScript [
   &EmitterState()
   &SpawnRate(SpawnRate=6, Use Spawn Probability=false)
  ]
  particleSpawnScript [
   /SetVariables(
    Particles.PlacementPicker:NiagaraInt32=&ModuloInt(
     A=&GenerateIntFromCounter(SimpleCounter:NiagaraDataInterfaceSimpleCounter=DataInterface() {propertyValues {GpuSyncMode=None, InitialValue=0}})
     B=3
    )
    Particles.SpawnGroup=&ModuloInt(
     A=&GenerateIntFromCounter(SimpleCounter:NiagaraDataInterfaceSimpleCounter=DataInterface() {propertyValues {GpuSyncMode=None, InitialValue=0}})
     B=3
    )
   ) {isSetParameters=true, module=SetVariables_#1}
   /SetVariables(Particles.PositionArray=DataInterface() {propertyValues {GpuSyncMode=SyncCpuToGpu, MaxElements=0, FloatData [<0 0 -7>, <0 6.5 4>, <0 -6.5 4>]}}) {isSetParameters=true, module=SetVariables_#2}
   /SetVariables(Particles.PositionOffset=DataInterface() {propertyValues {GpuSyncMode=SyncCpuToGpu, MaxElements=0, FloatData [<0 0 0>, <0 0.1 8>, <0 0.1 8>]}}) {isSetParameters=true, module=SetVariables_#3}
   /SetVariables(Particles.RotationArray=DataInterface() {propertyValues {GpuSyncMode=SyncCpuToGpu, MaxElements=0, FloatData [<0 0 1>, <0 -1 -1>, <0 1 -1>]}}) {isSetParameters=true, module=SetVariables_#4}
   &InitializeParticle(
    Color=<rgba 0.4005 0.1009 1 1>
    Lifetime Max=0.5
    Lifetime Min=0.4
    Lifetime Mode=::NewEnumerator1
    Sprite Size=&MakeVector2D(
     X=1.5
     Y=&Multiply_Float(A=&RandomRangeFloat(Minimum=6, Maximum=9), B=$User.RingSize)
    )
    Sprite Size Mode=::NewEnumerator4
    UsePositionOffset=false
    Write Material Random=true
   )
   &SphereLocation(
    MaskSpawnGroup=false
    Offset=&self:PositionOffset(
     ArrayIndex:NiagaraInt32=$Particles.SpawnGroup
     NewScale:NiagaraFloat=$User.NewScale
     PositionArray:NiagaraDataInterfaceArrayFloat3=$Particles.PositionArray
     PositionOffset:NiagaraDataInterfaceArrayFloat3=$Particles.PositionOffset
    )
    Override Seed=false
    Sphere Radius=1
    Surface Only=false
    UseNonUniformScale=false
    UseOffset=true
    UseSphereOrientation=false
   )
   &AddVelocityInCone(
    Cone Angle=80
    Cone Axis=&self:RotationOffset_01(
     ArrayIndex:NiagaraInt32=$Particles.SpawnGroup
     RotationArray:NiagaraDataInterfaceArrayFloat3=$Particles.RotationArray
     VecOffset:Vector3f=<0 0 0>
    )
    Module.Use Velocity Falloff On Cone Axis=true
    Velocity Distribution Along Cone Axis=1
    Velocity Falloff Away From Cone Axis=1
    Velocity Strength=2
   )
  ]
  particleUpdateScript [
   &SolveForcesAndVelocity(Clamp Velocity=false, Limit Acceleration=false)
   &ParticleState()
   &ScaleColor(
    Scale Alpha=&FloatFromCurve(
     FloatCurve=DataInterface() {propertyValues {Curve {keys [
         {arriveTangent=3.0956, interpMode=RCIM_Cubic, leaveTangent=3.0956, tangentMode=RCTM_User, time=0, value=0}
         <at 0.3 0.5>
         {arriveTangent=-0.0207, interpMode=RCIM_Cubic, leaveTangent=-0.0207, tangentMode=RCTM_User, time=0.5, value=0.5}
         <at 1 0>
        ]}}}
     Scale Curve=&RandomRangeFloat(Minimum=0.1, Maximum=0.1)
    )
    Scale RGB=&VectorFromFloat(Value=0.5)
    ScaleA=true
    ScaleRGB=true
   )
   &ScaleSpriteSize(Scale Factor=&MakeVector2D(
     X=1
     Y=&FloatFromCurve(FloatCurve=DataInterface() {propertyValues {Curve {keys [<at 0 0.2 RCIM_Cubic>, {interpMode=RCIM_Cubic, tangentMode=RCTM_User, time=1, value=1}]}}})
    ))
   &ScaleColor(Scale Alpha=&Lerp_Float(A=0, B=1, Alpha=$User.LaunchBrightness), ScaleA=true, ScaleRGB=false) {module=ScaleColor001}
  ]
  rendererClasses [&NiagaraSpriteRendererProperties]
  renderers [{rendererClass=&NiagaraSpriteRendererProperties}]
 }
 Ring {
  emitterUpdateScript [
   &EmitterState()
   &SpawnRate(SpawnRate=10, Use Spawn Probability=false)
  ]
  particleSpawnScript [
   &InitializeParticle(
    Color=<rgba 0.0673 1 0.6851 1>
    Lifetime Max=0.3
    Lifetime Min=0.2
    Lifetime Mode=::NewEnumerator1
    Mesh Scale Mode=::NewEnumerator3
    Mesh Uniform Scale=&Multiply_Float(A=0.06, B=1)
    UsePositionOffset=false
    Write Material Random=true
   )
   &InitialMeshOrientation(
    Orientation Vector=<0 0 1>
    Rotation=&RandomRangeVector(Minimum=<0 0 0>, Maximum=<0 0 1>)
    Use Orientation Vector=true
    Use Rotation Vector=true
   )
   &AddVelocity()
  ]
  particleUpdateScript [
   &SolveForcesAndVelocity(Clamp Velocity=false, Limit Acceleration=false)
   &ParticleState()
   &ScaleColor(
    Scale RGB=&Multiply_VectorByFloat(Vector=<1 1 1>, Float=&Lerp_Float(A=2, B=2.5, Alpha=$User.ADS))
    ScaleA=true
    ScaleRGB=true
   )
   &DynamicMaterialParameters(
    Index 0 Param 1=25
    Index 0 Param 2=0.25
    Index 0 Param 3=1.5
    Module.Param0WriteEnabled=true
    Module.Param1WriteEnabled=true
    Module.Param2WriteEnabled=true
    Module.Param3WriteEnabled=false
   )
   &ScaleMeshSize(Scale Factor=&Multiply_VectorByFloat(Vector=<1 1 1>, Float=$User.RingSize))
  ]
  rendererClasses [&NiagaraMeshRendererProperties]
  renderers [{rendererClass=&NiagaraMeshRendererProperties, rendererIndex=0}]
 }
 RingInX {
  emitterUpdateScript [
   &EmitterState()
   &SpawnRate(SpawnRate=6, Use Spawn Probability=false) {enabled=false}
   /Game/Effects/Niagara/Modules/Spawn/SpawnRateFromMultipleSpawnGroups_2(Global Spawnrate:NiagaraFloat=&Multiply_Float(
     A=&Lerp_Float(A=1, B=3, Alpha=&NormalizeFloat(Value=$User.RingSize, Min=0, Max=20))
     B=&Lerp_Float(A=1, B=2, Alpha=$User.ADS)
    ))
  ]
  particleSpawnScript [
   &FN_SetSpawnGroup(Number of Spawn Groups=2)
   &InitializeParticle(
    Lifetime Max=1
    Lifetime Mode=::NewEnumerator1
    Mesh Scale Mode=::NewEnumerator3
    Mesh Uniform Scale=&Multiply_Float(A=0.06, B=1)
    UsePositionOffset=false
    Write Material Random=true
   )
   &InitialMeshOrientation(
    Orientation Vector=<0 0 1>
    Rotation=&RandomRangeVector(Minimum=<0 0 0>, Maximum=<0 0 1>)
    Use Orientation Vector=true
    Use Rotation Vector=true
   )
   &AddVelocity(Velocity=&Multiply_Vector(
     A=&Multiply_VectorByFloat(Vector=<6 0 0>, Float=$User.RingSize)
     B=&VectorFromFloat(Value=/Niagara/DynamicInputs/Math/LerpMultipleFloats(
       Float Count:ENiagara_MultipleLerpCount=::NewEnumerator3
       1:NiagaraFloat=1
       2:NiagaraFloat=-1
       3:NiagaraFloat=?
       4:NiagaraFloat=?
       5:NiagaraFloat=?
       Alpha:NiagaraFloat=&MakeFloatFromInt(Int=$Particles.SpawnGroup)
      ))
    ))
  ]
  particleUpdateScript [
   &SolveForcesAndVelocity(Clamp Velocity=false, Limit Acceleration=false)
   &SolveRotationalForcesAndVelocity()
   &ParticleState()
   /SetVariables(Particles.SizeWave:NiagaraFloat=&Waveform(
     Global Input Position=&Multiply_Float(A=$Engine.Time, B=1)
     Global Phase=$Particles.NormalizedAge
     "Waveform [1]"=::NewEnumerator0
     "[1] Phase Scale"=1.75
     "[1] Phase"=$Particles.NormalizedAge
     "[1] Frequency"=&Lerp_Float(A=0.4, B=1.1, Alpha=$User.ADS)
     "[1] Amplitude Min/Max"=<0 1>
     "[1] Scale"=0.5
     "[1] Bias"=0.5
     "[2] Phase"=$Emitter.Age
     "[3] Phase"=$Emitter.Age
    )) {isSetParameters=true}
   &ScaleColor(
    Scale RGB=&MakeVectorFromLinearColorRGB(Color=&LerpLinearColors(
      StartColor=&MultiplyLinearColor(
       Linear Color=<rgba 0.1876 0.0825 0.75 1>
       Scale Factor=&MakeFloat4FromVectorAndFloat("Vector (RGB)"=&VectorFromFloat(Value=1), "Float (Alpha)"=1)
      )
      EndColor=&MultiplyLinearColor(
       Linear Color=<rgba 0 1 0.2061 1>
       Scale Factor=&MakeFloat4FromVectorAndFloat("Vector (RGB)"=&VectorFromFloat(Value=1), "Float (Alpha)"=1)
      )
      LerpFactor=&FloatFromCurve(FloatCurve=DataInterface() {propertyValues {Curve {keys [
           {arriveTangent=1.6292, interpMode=RCIM_Cubic, leaveTangent=1.6292, tangentMode=RCTM_User, time=0, value=0}
           {arriveTangent=0.0651, interpMode=RCIM_Cubic, tangentMode=RCTM_Break, time=1, value=1}
          ]}}})
     ))
    ScaleA=true
    ScaleRGB=true
   ) {module=ScaleColor001}
   &ScaleColor(
    Scale Alpha=&FloatFromCurve(
     FloatCurve=DataInterface() {propertyValues {Curve {keys [<at 0 1 RCIM_Cubic>, {interpMode=RCIM_Cubic, tangentMode=RCTM_Break, time=1, value=0}]}}}
     Scale Curve=3
    )
    Scale RGB=&Multiply_VectorByFloat(Vector=<1 1 1>, Float=&Lerp_Float(A=2, B=3, Alpha=$User.ADS))
    ScaleA=true
    ScaleRGB=true
   )
   &DynamicMaterialParameters(
    Index 0 Param 1=15
    Index 0 Param 2=0.25
    Index 0 Param 3=1.75
    Module.Param0WriteEnabled=true
    Module.Param1WriteEnabled=true
    Module.Param2WriteEnabled=true
    Module.Param3WriteEnabled=false
   )
   &ScaleMeshSize(Scale Factor=&Multiply_VectorByFloat(
     Vector=&Multiply_VectorByFloat(
      Vector=&MakeVector(X=$Particles.SizeWave, Y=$Particles.SizeWave, Z=1)
      Float=&FloatFromCurve(FloatCurve=DataInterface() {propertyValues {Curve {keys [
           {interpMode=RCIM_Cubic, tangentMode=RCTM_User, time=0, value=1}
           {arriveTangent=-0.5603, interpMode=RCIM_Cubic, tangentMode=RCTM_Break, time=1, value=0.8}
          ]}}})
     )
     Float=$User.RingSize
    ))
  ]
  rendererClasses [&NiagaraMeshRendererProperties]
  renderers [{rendererClass=&NiagaraMeshRendererProperties, rendererIndex=0}]
 }
 decal {
  emitterUpdateScript [
   &EmitterState()
   &SpawnRateFromMultipleSpawnGroups_2(Global Spawnrate=3)
  ]
  particleSpawnScript [
   &FN_SetSpawnGroup(Number of Spawn Groups=2)
   &InitializeParticle(
    Lifetime Max=0.75
    Lifetime Min=0.5
    Lifetime Mode=::NewEnumerator1
    Sprite Size Mode=::NewEnumerator1
    Uniform Sprite Size Max=200
    Uniform Sprite Size Min=150
    UsePositionOffset=false
    Write Material Random=true
   )
   &CylinderLocation(
    Cylinder Height=5
    Cylinder Radius=&Multiply_Float(A=1, B=$User.RingSize)
    MaskSpawnGroup=false
    Module.Enable Lathe Profile=false
    Module.Override Randomness=false
    Module.Surface Only=true
    Orientation Axis:ENiagaraOrientationAxis=::XAxis
    Use Endcaps In Surface Only Mode=false
   )
   &AddVelocity(Velocity=&Multiply_Vector(
     A=<75 0 0>
     B=&VectorFromFloat(Value=&LerpMultipleFloats(
       Float Count:ENiagara_MultipleLerpCount=::NewEnumerator3
       1:NiagaraFloat=-1
       2:NiagaraFloat=1
       3:NiagaraFloat=?
       4:NiagaraFloat=?
       5:NiagaraFloat=?
       Alpha:NiagaraFloat=&MakeFloatFromInt(Int=$Particles.SpawnGroup)
      ))
    ))
  ]
  particleUpdateScript [
   &SolveForcesAndVelocity(Clamp Velocity=false, Limit Acceleration=false)
   &SolveRotationalForcesAndVelocity()
   &ParticleState()
   &ScaleColor(
    Scale RGB=&MakeVectorFromLinearColorRGB(Color=&LerpLinearColors(
      StartColor=&MultiplyLinearColor(
       Linear Color=<rgba 0.1876 0.0825 0.75 1>
       Scale Factor=&MakeFloat4FromVectorAndFloat("Vector (RGB)"=&VectorFromFloat(Value=1), "Float (Alpha)"=1)
      )
      EndColor=&MultiplyLinearColor(
       Linear Color=<rgba 0 1 0.2061 1>
       Scale Factor=&MakeFloat4FromVectorAndFloat("Vector (RGB)"=&VectorFromFloat(Value=1), "Float (Alpha)"=1)
      )
      LerpFactor=&FloatFromCurve(FloatCurve=DataInterface() {propertyValues {Curve {keys [
           {arriveTangent=1.6292, interpMode=RCIM_Cubic, leaveTangent=1.6292, tangentMode=RCTM_User, time=0, value=0}
           {arriveTangent=0.0651, interpMode=RCIM_Cubic, tangentMode=RCTM_Break, time=1, value=1}
          ]}}})
     ))
    ScaleA=true
    ScaleRGB=true
   ) {module=ScaleColor001}
   &ScaleColor(
    Scale Alpha=&FloatFromCurve(FloatCurve=DataInterface() {propertyValues {Curve {keys [
         {interpMode=RCIM_Cubic, tangentMode=RCTM_User, time=0, value=0}
         <at 0.3 0.5>
         {arriveTangent=-0.0207, interpMode=RCIM_Cubic, leaveTangent=-0.0207, tangentMode=RCTM_User, time=0.7, value=0.5}
         <at 1 0>
        ]}}})
    Scale RGB=&Multiply_VectorByFloat(Vector=&VectorFromFloat(Value=1), Float=1)
    ScaleA=true
    ScaleRGB=true
   )
  ]
  rendererClasses [&NiagaraSpriteRendererProperties]
  renderers [{rendererClass=&NiagaraSpriteRendererProperties}]
 }
}
summary {
 systemName=NS_GravyGoblin_Holder
 emitters [
  {emitterName=Ring, rendererClasses [&NiagaraMeshRendererProperties]}
  {emitterName=RingInX, rendererClasses [&NiagaraMeshRendererProperties]}
  {emitterName=decal, rendererClasses [&NiagaraSpriteRendererProperties]}
  {emitterName=Glow, rendererClasses [&NiagaraSpriteRendererProperties, &NiagaraLightRendererProperties]}
  {emitterName=Flare, rendererClasses [&NiagaraSpriteRendererProperties]}
  {emitterName=GlowCone, rendererClasses [&NiagaraMeshRendererProperties]}
  {emitterName=Lines, rendererClasses [&NiagaraSpriteRendererProperties]}
 ]
 userVariables [
  User.ADS {type=NiagaraFloat, defaultValue=0, description=""}
  User.LaunchBrightness {type=NiagaraFloat, defaultValue=1, description=""}
  User.NewScale {type=NiagaraFloat, defaultValue=0, description=""}
  User.RingSize {type=NiagaraFloat, defaultValue=8, description=""}
 ]
}
