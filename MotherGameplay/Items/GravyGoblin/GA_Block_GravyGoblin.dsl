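; Gameplay ability for the Gravy Goblin gravity gun that forwards aim-down-sights input to the weapon's projectile holder.
; On activation it caches the player and current weapon, then starts parallel waits on targeting-tag add, tag removal and weapon-switch events.
; While the gun is in use, targeting start calls ADSPressed if ProjectileInUse is tagged, and targeting end calls ADSReleased.
; Only weapon switches set the in-use flag, refreshing the weapon reference.
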(event Ability|EventActivateAbility
  (bind (_owneractor _avataractor) (Utilities|Struct|BreakGameplayAbilityActorInfo (Ability|GetActorInfo)))
  (bind _weapon (Variables|Default|GetWeapon))
  (bind _holder (Variables|/MotherGameplay/Items/GravyGoblin/B_GravyGoblin_Athena.B_GravyGoblin_Athena_C|GetHolder _weapon))
  (bind _weapon_2 (Variables|Default|GetWeapon))
  (bind _holder_2 (Variables|/MotherGameplay/Items/GravyGoblin/B_GravyGoblin_Athena.B_GravyGoblin_Athena_C|GetHolder _weapon_2))
  (|CallParentFunction)
  (bind _asplayer_pawn_athena (Utilities|Casting|CastToPlayerPawn_Athena _avataractor))
  (Variables|Default|SetPlayer _asplayer_pawn_athena)
  (bind _asb_gravy_goblin_athena (Utilities|Casting|CastToB_GravyGoblin_Athena (Variables|/Script/FortniteGame.FortPawn|GetCurrentWeapon _asplayer_pawn_athena)))
  (Variables|Default|SetWeapon _asb_gravy_goblin_athena)
  (Utilities|FlowControl|Sequence
    (:then_0
      (Ability|Tasks|WaitGameplayTagAdd "(TagName=\"Gameplay.Status.IsTargeting\")"
        (:Added
          (if (Variables|Default|GetIsUsingGravityGun)
            (if (GameplayTags|HasMatchingGameplayTag (Variables|Default|GetPlayer) "(TagName=\"Granted.Athena.GravyGoblin.ProjectileInUse\")")
              (Utilities|IsValid _holder
                (:Is%20Valid
                  (Class|PrjGravyGoblinHolder|ADSPressed _holder))))))))
    (:then_1
      (Ability|Tasks|WaitGameplayTagRemove "(TagName=\"Gameplay.Status.IsTargeting\")"
        (:Removed
          (if (Variables|Default|GetIsUsingGravityGun)
            (Utilities|IsValid _holder_2
              (:Is%20Valid
                (Class|PrjGravyGoblinHolder|ADSReleased _holder_2)))))))
    (:then_3
      (bind (_asynctaskproxy _payload) (Ability|Tasks|WaitGameplayEvent "(TagName=\"Event.Weapons.WeaponSwitch\")")
        (:EventReceived
          (bind _output_get (Variables|Default|SetIsUsingGravityGun (.bSuccess (Utilities|Casting|CastToB_GravyGoblin_Athena (.OptionalObject (Utilities|Struct|BreakGameplayEventData _payload))))))
          (if _output_get
            (Variables|Default|SetWeapon (Utilities|Casting|CastToB_GravyGoblin_Athena (.OptionalObject (Utilities|Struct|BreakGameplayEventData _payload))))))))))

(event Ability|EventFailedToActivatePassiveAbility
  (params (ActorInfo :type Gameplay%20Ability%20Actor%20Info%20Structure)))
