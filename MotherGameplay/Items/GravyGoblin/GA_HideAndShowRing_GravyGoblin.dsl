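; Gameplay ability for the Gravy Goblin item that hides the weapon's ring while the ability is active.
; On activation it calls the parent, casts the avatar's current weapon to B_GravyGoblin_Athena, caches it and calls HideRing.
; On end, cancelled or not, it calls ShowRing on the cached weapon and applies the Gravy Goblin cooldown effect to the owner.
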(event Ability|EventActivateAbility
  (bind (_owneractor _avataractor) (Utilities|Struct|BreakGameplayAbilityActorInfo (Ability|GetActorInfo)))
  (|CallParentFunction)
  (bind _asb_gravy_goblin_athena (Utilities|Casting|CastToB_GravyGoblin_Athena (Variables|/Script/FortniteGame.FortPawn|GetCurrentWeapon (Utilities|Casting|CastToPlayerPawn_Athena _avataractor))))
  (bind _output_get (Variables|Default|SetAs%20B%20Gravy%20Goblin%20Athena _asb_gravy_goblin_athena))
  (Class|BGravyGoblinAthena|HideRing _output_get))

(event Ability|EventOnEndAbility
  (params (bWasCancelled :type Boolean))
  (|CallParentFunction bWasCancelled)
  (Class|BGravyGoblinAthena|ShowRing (Variables|Default|GetAs%20B%20Gravy%20Goblin%20Athena))
  (Ability|ApplyGameplayEffectToOwner "/MotherGameplay/Items/GravyGoblin/GE_Cooldown_GravyGoblin.GE_Cooldown_GravyGoblin_C"))

(event Ability|EventFailedToActivatePassiveAbility
  (params (ActorInfo :type Gameplay%20Ability%20Actor%20Info%20Structure)))
