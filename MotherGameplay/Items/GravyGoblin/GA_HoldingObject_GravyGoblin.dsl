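; Gameplay ability active while the Gravy Goblin gravity gun holds an object, caching the Prj_GravyGoblin_Holder from its granting effect's causer.
; Input press and release go to the holder's FirePressed and FireReleased while the gun is equipped.
; A weapon-switch event updates that equipped flag, dropping the object for other weapons.
; Holster and vehicle delegates also drop it, though vehicle occupants only when driving or untargeted in seats requiring targeting to shoot.
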(event Ability|EventActivateAbility
  (|CallParentFunction)
  (bind _asprj_gravy_goblin_holder (Utilities|Casting|CastToPrj_GravyGoblin_Holder (Ability|EffectContext|GetEffectCauser (Ability|GetGrantedbyEffectContext))))
  (Variables|Default|SetTheHolder _asprj_gravy_goblin_holder)
  (Utilities|FlowControl|Sequence
    (:then_0
      (Ability|Tasks|WaitInputPress false
        (:OnPress
          (Utilities|FlowControl|Sequence
            (:then_0
              _)
            (:then_1
              _))
          (if (Variables|Default|GetIsUsingGravityGun)
            (Class|PrjGravyGoblinHolder|FirePressed (Variables|Default|GetTheHolder))))))
    (:then_1
      (Ability|Tasks|WaitInputRelease false
        (:OnRelease
          (Utilities|FlowControl|Sequence
            (:then_0
              _)
            (:then_1
              _))
          (if (Variables|Default|GetIsUsingGravityGun)
            (Class|PrjGravyGoblinHolder|FireReleased (Variables|Default|GetTheHolder))))))
    (:then_2
      (bind (_asynctaskproxy _payload) (Ability|Tasks|WaitGameplayEvent "(TagName=\"Event.Weapons.WeaponSwitch\")")
        (:EventReceived
          (bind _output_get (Variables|Default|SetIsUsingGravityGun (.bSuccess (Utilities|Casting|CastToB_GravyGoblin_Athena (.OptionalObject (Utilities|Struct|BreakGameplayEventData _payload))))))
          (if (not _output_get)
            (Class|PrjGravyGoblinHolder|DropObject (Variables|Default|GetTheHolder))))))
    (:then_3
      (bind _asfort_player_pawn (Utilities|Casting|CastToFortPlayerPawn (Variables|/MotherGameplay/Items/GravyGoblin/Prj_GravyGoblin_Holder.Prj_GravyGoblin_Holder_C|GetOwningPlayer (Variables|Default|GetTheHolder))))
      (EventDispatchers|BindEventtoOnHolsteredEvent _asfort_player_pawn (delegate Custom|Drop))
      (Vehicle|BindEventtoOnVehicleStateChangedEvent _asfort_player_pawn (delegate Custom|VehicleStateChange)))))

(event Custom|VehicleStateChange
  (params (FortPlayerPawn :type Fort%20Player%20Pawn%20Object%20Reference) (NewVehicle :type Actor%20Object%20Reference) (OldVehicle :type Actor%20Object%20Reference))
  (if (Utilities|NotEqual(Object) NewVehicle 0)
    (bind _asfort_vehicle_interface (Utilities|Casting|CastToFortVehicleInterface NewVehicle))
    (CallFunction|CheckforDrop FortPlayerPawn _asfort_vehicle_interface)))

(event Custom|CheckForDrop
  (params (PawnToCheck :type Fort%20Player%20Pawn%20Object%20Reference) (Vehicle :type Fort%20Vehicle%20Interface%20Interface))
  (if (or (Vehicle|IsDrivingVehicle PawnToCheck) (and (Vehicle|IsPawninSeatRequiringTargetingforShooting Vehicle PawnToCheck) (not (GameplayTags|HasMatchingGameplayTag PawnToCheck "(TagName=\"Gameplay.Status.IsTargeting\")"))))
    (CallFunction|DropServer)
    (Class|PrjGravyGoblinHolder|DropObject (Variables|Default|GetTheHolder))))

(event Custom|Drop
  (bind _theholder (Variables|Default|GetTheHolder))
  (bind _owningplayer (Variables|/MotherGameplay/Items/GravyGoblin/Prj_GravyGoblin_Holder.Prj_GravyGoblin_Holder_C|GetOwningPlayer _theholder))
  (bind _asfort_player_pawn (Utilities|Casting|CastToFortPlayerPawn _owningplayer))
  (bind _vehicle_occupied (Utilities|IsinVehicle _asfort_player_pawn)
    (:true
      (CallFunction|CheckforDrop _asfort_player_pawn _vehicle_occupied))
    (:false
      (CallFunction|DropServer)
      (Class|PrjGravyGoblinHolder|DropObject (Variables|Default|GetTheHolder)))))

(event Custom|DropServer
  (Class|PrjGravyGoblinHolder|DropObject (Variables|Default|GetTheHolder)))

(event Ability|EventFailedToActivatePassiveAbility
  (params (ActorInfo :type Gameplay%20Ability%20Actor%20Info%20Structure)))
