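; Ranged gameplay ability for the Gravy Goblin item that gates when it can activate.
; CanActivateAbility refuses activation while the avatar actor has either the projectile-in-use tag or the fire disk cooldown tag.
; The end-ability event is only a disabled stub, and the graphs shown hold no firing or projectile logic.
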(event Ability|EventOnEndAbility
  (params (bWasCancelled :type Boolean))
  (|CallParentFunction bWasCancelled) ; This node is disabled and will not be called.\nDrag off pins to build functionality.
)

(fn CanActivateAbility
  (params (ActorInfo_OwnerActor :type Actor%20Object%20Reference) (ActorInfo_AvatarActor :type Actor%20Object%20Reference) (ActorInfo_PlayerController :type Player%20Controller%20Object%20Reference) (ActorInfo_AbilitySystemComponent :type Ability%20System%20Component%20Object%20Reference) (ActorInfo_SkeletalMeshComponent :type Skeletal%20Mesh%20Component%20Object%20Reference) (ActorInfo_AnimInstance :type Anim%20Instance%20Object%20Reference) (ActorInfo_MovementComponent :type Movement%20Component%20Object%20Reference) (ActorInfo_AffectedAnimInstanceTag :type Name) (Handle :type Gameplay%20Ability%20Spec%20Handle%20Structure))
  (bind _returnvalue (GameplayTags|HasAnyMatchingGameplayTags(Message) ActorInfo_AvatarActor (GameplayTags|MakeLiteralGameplayTagContainer "(GameplayTags=((TagName=\"Granted.Athena.GravyGoblin.ProjectileInUse\"),(TagName=\"Granted.Cooldown.Mother.GravyGoblin.FireDiskCooldown\")))")))
  (return 0 (not _returnvalue)))
