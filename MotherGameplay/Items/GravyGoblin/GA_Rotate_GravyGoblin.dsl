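; Gameplay ability that lets the player rotate the Gravy Goblin holder while wielding the Gravy Goblin gravity gun.
; On activation it caches the holder projectile from the effect causer of the effect that granted the ability.
; A weapon switch event task sets a flag by casting the event's optional object to the Gravy Goblin item.
; Input presses call RotatePressed on the holder when that flag is set, otherwise printing a debug string.
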(event Ability|EventActivateAbility
  (|CallParentFunction)
  (bind _asprj_gravy_goblin_holder (Utilities|Casting|CastToPrj_GravyGoblin_Holder (Ability|EffectContext|GetEffectCauser (Ability|GetGrantedbyEffectContext))))
  (Variables|Default|SetTheHolder _asprj_gravy_goblin_holder)
  (Utilities|FlowControl|Sequence
    (:then_0
      (Ability|Tasks|WaitInputPress false
        (:OnPress
          (Utilities|FlowControl|Sequence
            (:then_0
              _)
            (:then_1
              _))
          (if (Variables|Default|GetIsUsingGravityGun)
            (Class|PrjGravyGoblinHolder|RotatePressed (Variables|Default|GetTheHolder))
            (Development|PrintString "PRESSED")))))
    (:then_2
      (bind (_asynctaskproxy _payload) (Ability|Tasks|WaitGameplayEvent "(TagName=\"Event.Weapons.WeaponSwitch\")")
        (:EventReceived
          (bind _output_get (Variables|Default|SetIsUsingGravityGun (.bSuccess (Utilities|Casting|CastToB_GravyGoblin_Athena (.OptionalObject (Utilities|Struct|BreakGameplayEventData _payload))))))
          (if _output_get
            _))))))
