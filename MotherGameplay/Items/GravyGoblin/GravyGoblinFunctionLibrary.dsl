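; Function library for the GravyGoblin (Grabitron) gravity gun, holding a deprecated target validity check superseded by IsValidTarget in GravityGunTargetTracking.
; The check rejects players, wrong-sized, already-grabbed or blacklisted actors and the owner's own vehicle, returning failure tags for some cases.
; Data-row toggles gate large objects, multiple grabbers, vehicles, pickups and projectiles, while physics-simulating actors also qualify.
; IsGrabbablePickup accepts only pickups without an attach parent.
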(fn CanBePickedUpByGrabitron_DEPRECATED
  (params (OwningPlayer :type Actor%20Object%20Reference) (Actor :type Actor%20Object%20Reference))
  (bind (_asfort_pickup _bsuccess) (Utilities|Casting|CastToFortPickup (LocalVariables|GetActor_)))
  (bind _actor (LocalVariables|GetActor_))
  (bind (_asfort_projectile_base _bsuccess_2) (Utilities|Casting|CastToFortProjectileBase _actor))
  ;[1 1 1] DO NOT USE - Instead use IsValidTarget in the GravityGunTargetTracking component
  (LocalVariables|SetOwningPlayer_ OwningPlayer)
  (bind _output_get (LocalVariables|SetActor_ Actor))
  (Utilities|Casting|CastToFortPlayerPawn _output_get
    (:then
      (return false "(GameplayTags=((TagName=\"Athena.Ability.Failed.GravyGoblin.TargetIsPlayer\")))"))
    (:CastFailed
      (bind _returnvalue (FortPhysicsSimulationLibrary|DoesPhysicsObjectMeetSizeRequirement (LocalVariables|GetActor_) (select (Class|AthenaFunctionLibrary|RowtoBool (LocalVariables|GetAllowLargeObjects) 0.0 (LocalVariables|GetOwningPlayer_)) "Large" "Medium")))
      (if _returnvalue
        (if (and (Utilities|Array|ContainsItem (Variables|/Script/Engine.Actor|GetTags (LocalVariables|GetActor_)) "GrabbedByGravyGoblin") (or (not (Class|AthenaFunctionLibrary|RowtoBool (LocalVariables|GetAllowMultipleGrabbers))) _bsuccess))
          (return false "(GameplayTags=((TagName=\"Athena.Ability.Failed.GravyGoblin.TargetIsAlreadyGrabbed\")))")
          (else
            (bind _asfort_vehicle_interface (Utilities|Casting|CastToFortVehicleInterface (LocalVariables|GetActor_))
              (:then
                (if (Class|AthenaFunctionLibrary|RowtoBool (LocalVariables|GetAllowVehicles) 0.0 (LocalVariables|GetOwningPlayer_))
                  (bind _asfort_player_pawn (Utilities|Casting|CastToFortPlayerPawn (LocalVariables|GetOwningPlayer_))
                    (:then
                      (bind _vehicle_occupied (Utilities|IsinVehicle _asfort_player_pawn)
                        (:true
                          (if (Utilities|Equal(Object) (Vehicle|GetActor _asfort_vehicle_interface) (Vehicle|GetActor _vehicle_occupied))
                            (return false "(GameplayTags=((TagName=\"Athena.Ability.Failed.GravyGoblin.TargetIsOwnVehicle\")))")
                            (else
                              _)))
                        (:false
                          _))
                      (bind _returnvalue_2 (GameplayTags|HasAnyMatchingGameplayTags(Message) (LocalVariables|GetActor_) (GameplayTags|MakeLiteralGameplayTagContainer "(GameplayTags=((TagName=\"Granted.Athena.GravyGoblin.Blacklisted\"),(TagName=\"Granted.Athena.PhysicsObject.Blacklisted\")))")))
                      (if _returnvalue_2
                        (return false "(GameplayTags=)")
                        (else
                          (return true 0))))
                    (:CastFailed
                      _))
                  (else
                    (return false "(GameplayTags=)"))))
              (:CastFailed
                (bind _returnvalue_3 (FortPhysicsSimulationLibrary|CanActorSimulatePhysics _actor))
                (if (or _returnvalue_3 (and (CallFunction|IsGrabbablePickup _actor _actor) (Class|AthenaFunctionLibrary|RowtoBool (LocalVariables|GetAllowPickups) 0.0 (LocalVariables|GetOwningPlayer_))))
                  _
                  (elif (and _bsuccess_2 (Class|AthenaFunctionLibrary|RowtoBool (LocalVariables|GetAllowProjectiles) 0.0 (LocalVariables|GetOwningPlayer_)))
                    _
                    (else
                      (return false "(GameplayTags=)"))))))))
        (else
          (return false "(GameplayTags=)"))))))

(fn IsGrabbablePickup
  (params (Object :type Object%20Reference))
  (bind _asfort_pickup (Utilities|Casting|CastToFortPickup Object)
    (:then
      (return (not (Utilities|IsValid (Actor|GetAttachParentActor _asfort_pickup) :!class /Script/Engine.KismetSystemLibrary))))
    (:CastFailed
      (return false))))
