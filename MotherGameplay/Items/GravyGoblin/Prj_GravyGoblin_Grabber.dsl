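; Grabber projectile for the GravyGoblin gravity gun that flies out to grab objects, pickups, vehicles and projectiles, or yank players toward the shooter.
; Outbound, it can home on a target, using manual velocity steering for projectiles.
; Valid targets trigger a server-spawned Holder, initialized through OnReps with the target or a building-actor clone.
; On timeout or failed grab, it homes back to the weapon, then timeline-lerps into its ring.
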(event EventDestroyed
  (bind _returnvalue (CallFunction|GetOwningPlayer))
  ;[1 1 1] On Destroyed
  (|CallParentFunction)
  (if (not (Variables|Default|GetSpawnedHolder?))
    (CallFunction|StopinAirAudio true)
    (Utilities|IsValid _returnvalue
      (:Is%20Valid
        (GameplayEffects|RemoveActiveEffectswithGrantedTags (Ability|GetFortAbilitySystemComponent _returnvalue) "(GameplayTags=((TagName=\"Granted.Athena.GravyGoblin.ProjectileInUse\")))")))))

(event Custom|Start%20In%20Air%20Audio%20Server
  ;[1 1 1] On Begin Play
  (Default|StartinAirAudio))

(event EventBeginPlay
  ;[1 1 1] On Begin Play
  (|CallParentFunction)
  (Utilities|FlowControl|DoOnce false
    (:Completed
      (Utilities|Time|SetTimerbyEvent (delegate Custom|TurnAround) (ScalableFloat|GetValueatLevel (Variables|Design|GetTurnaroundTime)))
      (Variables|Default|SetOriginalVelocity (Math|Vector|VectorLength (Transformation|GetVelocity)))
      (Physics|SetPhysicsAngularVelocityinDegrees (Variables|Collision|GetCapsuleComponent) "0, 0,200.000000" false "None")
      (GameplayEffects|ApplyGameplayEffectToTarget :self (Ability|GetAbilitySystemComponent (CallFunction|GetOwningPlayer)) :GameplayEffectClass "/MotherGameplay/Items/GravyGoblin/GE_ProjectileInUse_GravyGoblin.GE_ProjectileInUse_GravyGoblin_C" :Target (Ability|GetAbilitySystemComponent (CallFunction|GetOwningPlayer)) :Level 0.0)
      (Networking|SwitchHasAuthority
        (:Remote
          (CallFunction|StartinAirAudioServer))) ; Needs to only happen if beginplay happens on client (sometimes grabber is destroyed before beginplay can run on client)\nOtherwise the sound won't be destroyed properly, and sound will get left behind
    )))

(event Custom|ForceTurnaround
  ;[0.4177 0.4177 0.4177] On Turn-Around (when prj starts coming back)
  (if (not (Variables|Default|GetTurnedAround))
    (Variables|Default|SetTurnedAround true)
    (Collision|SetCollisionEnabled (Variables|Collision|GetCapsuleComponent) "NoCollision")
    (Projectile|SetHomingStyle 0 0 "LockOn" false :!class /Script/FortniteGame.FortProjectileMovementInterface)
    (Utilities|FlowControl|Sequence
      (:then_0
        ;;[1 1 1] Set to home in on player's weapon
        (Networking|SwitchHasAuthority
          (:Authority
            (Projectile|SetHomingTurnSpeedMin 0 self 1e+04 :!class /Script/FortniteGame.FortProjectileMovementInterface)
            (Projectile|SetHomingTurnSpeedMax 0 self 1e+04 :!class /Script/FortniteGame.FortProjectileMovementInterface)
            (Projectile|SetHomingTimeUntilMaxTurnSpeed 0 self 0.0 :!class /Script/FortniteGame.FortProjectileMovementInterface)
            (bind _asplayer_pawn_athena (Utilities|Casting|CastToPlayerPawn_Athena (CallFunction|GetOwningPlayer)))
            (Projectile|SetHomingTarget 0 self (Variables|/Script/FortniteGame.FortPawn|GetCurrentWeapon _asplayer_pawn_athena) 0 :!class /Script/FortniteGame.FortProjectileMovementInterface))))
      (:then_1
        (bind _asplayer_pawn_athena_2 (Utilities|Casting|CastToPlayerPawn_Athena (CallFunction|GetOwningPlayer)))
        (bind _asb_gravy_goblin_athena (Utilities|Casting|CastToB_GravyGoblin_Athena (Variables|/Script/FortniteGame.FortPawn|GetCurrentWeapon _asplayer_pawn_athena_2)))
        (Variables|Default|SetWeapon _asb_gravy_goblin_athena)
        (Variables|Default|SetWeaponRing (Variables|/MotherGameplay/Items/GravyGoblin/B_GravyGoblin_Athena.B_GravyGoblin_Athena_C|GetRing _asb_gravy_goblin_athena))))))

(event Custom|TurnAround
  ;[0.4177 0.4177 0.4177] On Turn-Around (when prj starts coming back)
  (if (or (Variables|Default|GetBuiltinHoming) (Variables|Default|GetCustomHoming))
    (Utilities|Time|SetTimerbyEvent (delegate Custom|ForceTurnaround) (ScalableFloat|GetValueatLevel (Variables|Design|GetExtraTurnaroundTimeWhenHoming)))
    (elif (Variables|Default|GetTurnedAround)
      _
      (else
        (Variables|Default|SetTurnedAround true)
        (Collision|SetCollisionEnabled (Variables|Collision|GetCapsuleComponent) "NoCollision")
        (Projectile|SetHomingStyle 0 0 "LockOn" false :!class /Script/FortniteGame.FortProjectileMovementInterface)
        (Utilities|FlowControl|Sequence
          (:then_0
            ;;[1 1 1] Set to home in on player's weapon
            (Networking|SwitchHasAuthority
              (:Authority
                (Projectile|SetHomingTurnSpeedMin 0 self 1e+04 :!class /Script/FortniteGame.FortProjectileMovementInterface)
                (Projectile|SetHomingTurnSpeedMax 0 self 1e+04 :!class /Script/FortniteGame.FortProjectileMovementInterface)
                (Projectile|SetHomingTimeUntilMaxTurnSpeed 0 self 0.0 :!class /Script/FortniteGame.FortProjectileMovementInterface)
                (bind _asplayer_pawn_athena (Utilities|Casting|CastToPlayerPawn_Athena (CallFunction|GetOwningPlayer)))
                (Projectile|SetHomingTarget 0 self (Variables|/Script/FortniteGame.FortPawn|GetCurrentWeapon _asplayer_pawn_athena) 0 :!class /Script/FortniteGame.FortProjectileMovementInterface))))
          (:then_1
            (bind _asplayer_pawn_athena_2 (Utilities|Casting|CastToPlayerPawn_Athena (CallFunction|GetOwningPlayer)))
            (bind _asb_gravy_goblin_athena (Utilities|Casting|CastToB_GravyGoblin_Athena (Variables|/Script/FortniteGame.FortPawn|GetCurrentWeapon _asplayer_pawn_athena_2)))
            (Variables|Default|SetWeapon _asb_gravy_goblin_athena)
            (Variables|Default|SetWeaponRing (Variables|/MotherGameplay/Items/GravyGoblin/B_GravyGoblin_Athena.B_GravyGoblin_Athena_C|GetRing _asb_gravy_goblin_athena))))))))

(event EventTick
  (params (DeltaSeconds :type Float%20%28single-precision%29))
  (bind _returnvalue (Transformation|GetActorLocation))
  (bind _homingtarget (Variables|Default|GetHomingTarget))
  (bind _returnvalue_2 (Transformation|GetActorLocation _homingtarget))
  (bind _returnvalue_3 (- _returnvalue_2 _returnvalue))
  (bind _ring1 (Variables|Default|GetRing1))
  (bind _returnvalue_4 (Transformation|GetWorldLocation _ring1))
  (bind _returnvalue_5 (Math|Vector|VectorLength _returnvalue_3))
  (bind _homing_distancetoattach (Variables|Design|GetHoming_DistanceToAttach))
  (bind _returnvalue_6 (ScalableFloat|GetValueatLevel _homing_distancetoattach))
  (Variables|Default|SetCachedDT DeltaSeconds)
  (Utilities|FlowControl|Sequence
    (:then_0
      ;[1 1 1] Update Audio and VFX
      (if (not (Variables|Default|GetSpawnedHolder?))
        (CallFunction|UpdateinAirAudio)
        (bind _asplayer_pawn_athena (Utilities|Casting|CastToPlayerPawn_Athena (CallFunction|GetOwningPlayer)))
        (bind _asb_gravy_goblin_athena (Utilities|Casting|CastToB_GravyGoblin_Athena (Variables|/Script/FortniteGame.FortPawn|GetCurrentWeapon _asplayer_pawn_athena)))
        (Effects|Components|ParticleSystem|SetVectorParameter (Variables|/MotherGameplay/Items/GravyGoblin/B_GravyGoblin_Athena.B_GravyGoblin_Athena_C|GetNS_GravyGoblin_Beam1 _asb_gravy_goblin_athena) "SplineUpdatedEndPos" _returnvalue_4)
        (Effects|Components|ParticleSystem|SetVectorParameter (Variables|/MotherGameplay/Items/GravyGoblin/B_GravyGoblin_Athena.B_GravyGoblin_Athena_C|GetNS_GravyGoblin_Beam1 _asb_gravy_goblin_athena) "SplineUpdatedEndTangent" _returnvalue_4)))
    (:then_1
      (if (Variables|Default|GetTurnedAround)
        ;[0.4385 0.4385 0.4385] Update -- While Coming Back
        (if (< (Math|Vector|VectorLengthSquared (- (Transformation|GetActorLocation (CallFunction|GetOwningPlayer)) (Transformation|GetActorLocation))) (Math|Float|Square (ScalableFloat|GetValueatLevel (Variables|Design|GetDistanceToStartInterping))))
          (Utilities|FlowControl|DoOnce false
            (:Completed
              (Projectile|SetHomingStyle 0 self "None" true :!class /Script/FortniteGame.FortProjectileMovementInterface)
              (Variables|Default|SetCached_SpeedBeforeInterp (Math|Vector|VectorLength (Variables|/Script/Engine.MovementComponent|GetVelocity (Variables|Movement|GetProjectileMovementComponent))))
              (Projectile|SetVelocity (Variables|Movement|GetProjectileMovementComponent) (0.0 0.0 0.0))
              (bind (_direction _alpha) (Timeline|MoveToWeapon PlayFromStart)
                (:Update
                  (if (not (Utilities|CloseEnough?))
                    ;;[1 1 1] Manually interpolate directly into the ring slot (makes it lock into weapon more smoothly/accurately)
                    (Transformation|SetActorLocation (Math|Vector|Lerp(Vector) (Transformation|GetActorLocation) (Transformation|GetWorldLocation (Variables|Default|GetWeaponRing)) _alpha) false true)
                    (Transformation|SetWorldRotation (Variables|Default|GetRing1) (Math|Rotator|Lerp(Rotator) (Transformation|GetWorldRotation (Variables|Default|GetRing1)) (Transformation|GetWorldRotation (Variables|Default|GetWeaponRing)) _alpha true) false false)))
                (:Finished
                  _))
              ;;[1 1 1] Destroy and stop interping once close enough
              (Actor|DestroyActor)
              (Rendering|SetActorHiddenInGame true)
              (CallFunction|StopInterping)
              (CallFunction|StopinAirAudio true))))
        (elif (Variables|Default|GetCustomHoming)
          ;[1 1 1] Update -- while going toward object
          (if (< _returnvalue_5 _returnvalue_6)
            (Variables|Default|SetCustomHoming false)
            (if (Collision|GetActorEnableCollision (Variables|Default|GetHomingTarget))
              (CallFunction|SpawnHolder (Variables|Default|GetHomingTarget))
              (else
                (CallFunction|TurnAround)))
            (else
              (Networking|SwitchHasAuthority
                (:Authority
                  (Projectile|SetVelocity (Variables|Movement|GetProjectileMovementComponent) (* (Math|Vector|Normalize _returnvalue_3) (ScalableFloat|GetValueatLevel (Variables|Design|GetSpeedWhileHoming))))))))
          (elif (Variables|Default|GetBuiltinHoming)
            (if (< _returnvalue_5 (* _returnvalue_6 2.5))
              (Variables|Default|SetBuiltinHoming false))))))))

(event Custom|Stop%20Interping
  ;[0.4385 0.4385 0.4385] Update -- While Coming Back
  (Timeline|MoveToWeapon Stop))

(event Collision|EventActorBeginOverlap
  (params (OtherActor :type Actor%20Object%20Reference))
  ;[1 1 1] On Overlap (for PickUps)
  (if (not (Variables|Default|GetSpawnedHolder?))
    (bind _asfort_pickup (Utilities|Casting|CastToFortPickup OtherActor))
    (if (or (Variables|Default|GetBuiltinHoming) (Variables|Default|GetCustomHoming))
      (if (Utilities|Equal(Object) (Variables|Default|GetHomingTarget) OtherActor)
        _))
    (bind (_reasonforfailure _returnvalue) (Targeting|IsValidTarget (Variables|Default|GetGravityGunTargetTracking) (Variables|Actor|GetInstigator) _asfort_pickup :!class /Script/MotherGameplayRuntime.GravityGunTargetTrackingComponent))
    (if _returnvalue
      (CallFunction|SpawnHolder _asfort_pickup)
      (else
        (bind _asfort_player_pawn (Utilities|Casting|CastToFortPlayerPawn (Variables|Actor|GetInstigator)))
        (FailedAction|UnabletoPerformAction _asfort_player_pawn false _reasonforfailure)))))

(event OnComponentHit(CapsuleComponent)
  (params (HitComponent :type Primitive%20Component%20Object%20Reference) (OtherActor :type Actor%20Object%20Reference) (OtherComp :type Primitive%20Component%20Object%20Reference) (NormalImpulse :type Vector :default "0, 0, 0") (Hit :type Hit%20Result%20Structure%20%28by%20ref%29))
  (bind (_bblockinghit _binitialoverlap _time _distance _location _impactpoint) (Collision|BreakHitResult Hit))
  (if (not (Variables|Default|GetSpawnedHolder?))
    (Utilities|IsValid (Variables|Default|GetGrabbedObject)
      (:Is%20Not%20Valid
        (bind _asplayer_pawn_athena (Utilities|Casting|CastToPlayerPawn_Athena OtherActor)
          (:then
            ;[1 1 1] Yanking Players
            (bind _asfort_player_pawn (Utilities|Casting|CastToFortPlayerPawn (Variables|Actor|GetInstigator))
              (:then
                (Utilities|LaunchPlayerPawn _asfort_player_pawn _asplayer_pawn_athena (* (Math|Vector|vector+vector (Math|Vector|Normalize (Math|Vector|ProjectVectorontoPlane (- (Transformation|GetActorLocation _asfort_player_pawn) (Transformation|GetActorLocation _asplayer_pawn_athena)) "0, 0,1.000000")) (Math|Vector|MakeVector 0.0 0.0 (ScalableFloat|GetValueatLevel (Variables|Design|GetPlayerPullForce_UpMultiplier)))) (ScalableFloat|GetValueatLevel (Variables|Design|GetPlayerPullForce))) false false false true true))
              (:CastFailed
                _))
            (CallFunction|ForceTurnaround)
            (FailedAction|UnabletoPerformAction _asfort_player_pawn false (Variables|Default|GetPlayer%20Failed%20Reason)))
          (:CastFailed
            (bind (_reasonforfailure _returnvalue) (Targeting|IsValidTarget (Variables|Default|GetGravityGunTargetTracking) (CallFunction|GetOwningPlayer) OtherActor :!class /Script/MotherGameplayRuntime.GravityGunTargetTrackingComponent))
            (if _returnvalue
              (CallFunction|SpawnHolder OtherActor)
              (else
                (bind _asfort_player_pawn_2 (Utilities|Casting|CastToFortPlayerPawn (Variables|Actor|GetInstigator)))
                (FailedAction|UnabletoPerformAction _asfort_player_pawn_2 false _reasonforfailure) ; show reason for failing on screen if one is given
                ;[1 1 1] Check for projectiles on the ground
                (bind (_outactors _returnvalue_2) (Collision|SphereOverlapActors _impactpoint 6e+01 (Utilities|Array|MakeArray "ObjectTypeQuery8") "/Script/FortniteGame.FortProjectileBase" (Utilities|Array|MakeArray self)))
                (if _returnvalue_2
                  (CallFunction|SpawnHolder (Utilities|Array|Get(acopy) _outactors))
                  (else
                    (CallFunction|ForceTurnaround)))))))))))

(event Custom|SpawnHolder_
  (params (GrabbedObject :type Actor%20Object%20Reference))
  ;[1 1 1] Spawn Holder
  (bind (_reasonforfailure _returnvalue) (Targeting|IsValidTarget (Variables|Default|GetGravityGunTargetTracking) (Variables|Actor|GetInstigator) GrabbedObject :!class /Script/MotherGameplayRuntime.GravityGunTargetTrackingComponent))
  (if _returnvalue
    (bind _output_get (Variables|Default|SetGrabbedObject GrabbedObject))
    (Rendering|SetActorHiddenInGame true)
    (CallFunction|SpawnHolderServer _output_get)
    (Variables|Default|SetSpawnedHolder? true)
    (CallFunction|StopinAirAudio false)
    (else
      (CallFunction|ForceTurnaround))))

(event Custom|SpawnHolderServer
  (params (GrabbedObject :type Actor%20Object%20Reference))
  (bind (_asbuilding_actor _bsuccess) (Utilities|Casting|CastToBuildingActor (Variables|Default|GetGrabbedObject)))
  ;[1 1 1] Spawn Holder
  (Utilities|FlowControl|DoOnce false
    (:Completed
      (Rendering|SetActorHiddenInGame true)
      (bind _output_get (Variables|Default|SetGrabbedObject GrabbedObject))
      (Networking|SwitchHasAuthority
        (:Authority
          (bind _returnvalue (Game|SpawnActorfromClass "/MotherGameplay/Items/GravyGoblin/Prj_GravyGoblin_Holder.Prj_GravyGoblin_Holder_C" (Math|Transform|MakeTransform (Collision|GetActorBounds _output_get true false) (Transformation|GetActorRotation _output_get)) "AlwaysSpawn" "MultiplyWithRoot" 0 0 (Utilities|Casting|CastToPawn (CallFunction|GetOwningPlayer))))
          (Variables|Default|SetSpawnedHolder _returnvalue)
          (Projectile|SetHomingStyle 0 self "None" true :!class /Script/FortniteGame.FortProjectileMovementInterface)
          (Projectile|SetVelocity (Variables|Movement|GetProjectileMovementComponent) (0.0 0.0 0.0))
          (Variables|Default|SetSpawnedHolder? true)
          (if _bsuccess
            (bind _returnvalue_2 (GravityHolder|AttemptCloneActortoHold (Variables|Default|GetSpawnedHolder) (Variables|Default|GetGrabbedObject)))
            (Variables|Default|SetClonedObject _returnvalue_2)
            (else
              (Variables|Default|SetClonedObject (Variables|Default|GetGrabbedObject)))))))))

(fn GetOwningPlayer
  (return (Variables|Actor|GetInstigator)))

(fn GetPrjInfo
  (bind _asfort_projectile_base (Utilities|Casting|CastToFortProjectileBase (Variables|Default|GetGrabbedObject))
    (:then
      (return (Utilities|GetClass _asfort_projectile_base) (Variables|/Script/FortniteGame.FortGameplayEffectDeliveryActor|GetEffectContainerSpecToApplyOnTouch _asfort_projectile_base) (Variables|/Script/FortniteGame.FortGameplayEffectDeliveryActor|GetEffectContainerSpecToApplyOnExplode _asfort_projectile_base) (Variables|/Script/FortniteGame.FortGameplayEffectDeliveryActor|GetAdditionalExplodeEffectContainers _asfort_projectile_base)))
    (:CastFailed
      (return 0 0 0 0))))

(fn GetFortProjectileMovementComponent
  (return (Variables|Movement|GetProjectileMovementComponent)))

(fn InitializeHolder
  (bind (_class _effectcontainerspectoapplyontouch _effectcontainerspectoapplyonexplode _additionalexplodeeffectcontainers) (CallFunction|GetPrjInfo))
  (bind (_asfort_vehicle_interface _bsuccess) (Utilities|Casting|CastToFortVehicleInterface (Variables|Default|GetClonedObject)))
  (bind (_asfort_projectile_base _bsuccess_2) (Utilities|Casting|CastToFortProjectileBase (Variables|Default|GetClonedObject)))
  (bind (_asfort_pickup _bsuccess_3) (Utilities|Casting|CastToFortPickup (Variables|Default|GetClonedObject)))
  (Variables|Default|SetSpawnedHolder? true)
  (Utilities|Array|AddUnique (Variables|/Script/Engine.Actor|GetTags (Variables|Default|GetClonedObject)) "GrabbedByGravyGoblin")
  (Class|PrjGravyGoblinHolder|Initialize (Variables|Default|GetSpawnedHolder) (Variables|Default|GetClonedObject) (CallFunction|GetOwningPlayer) (select _bsuccess_2 "Projectile" (select _bsuccess_3 "PickUp" (select _bsuccess "Vehicle" "PhysicsObject"))) _class _effectcontainerspectoapplyontouch _effectcontainerspectoapplyonexplode _additionalexplodeeffectcontainers)
  (CallFunction|StopinAirAudio false)
  (Actor|SetLifeSpan 1.0)
  (Rendering|SetActorHiddenInGame true))

(fn OnRep_ClonedObject
  (Utilities|IsValid (Variables|Default|GetSpawnedHolder)
    (:Is%20Valid
      (CallFunction|InitializeHolder))
    (:Is%20Not%20Valid
      (Variables|Default|SetSpawnedHolderIsLate true))))

(fn OnRep_HomingTarget
  (Utilities|Casting|CastToFortProjectileBase (Variables|Default|GetHomingTarget)
    (:then
      (Variables|Default|SetCustomHoming true))
    (:CastFailed
      (Variables|Default|SetBuiltinHoming true))))

(fn OnRep_SpawnedHolder
  (if (Variables|Default|GetSpawnedHolderIsLate)
    (CallFunction|InitializeHolder)))

(fn OnRep_Test)

(fn SetHomingVars
  (params (HomingTarget :type Actor%20Object%20Reference))
  (bind _self self)
  (bind _homingtarget (Variables|Default|GetHomingTarget))
  (Utilities|IsValid HomingTarget
    (:Is%20Valid
      (bind _output_get (Variables|Default|SetHomingTarget HomingTarget))
      (Utilities|Casting|CastToFortProjectileBase _output_get
        (:then
          (Variables|Default|SetCustomHoming true))
        (:CastFailed
          (Variables|Default|SetBuiltinHoming true)
          (Projectile|SetHomingStyle 0 self "LockOn" true :!class /Script/FortniteGame.FortProjectileMovementInterface)
          (Networking|SwitchHasAuthority
            (:Authority
              ;[1 1 1] TODO: Make hotfixable values.
              (Projectile|SetHomingTurnSpeedMin 0 _self 1e+01 :!class /Script/FortniteGame.FortProjectileMovementInterface)
              (Projectile|SetHomingTurnSpeedMax 0 _self 1.2e+02 :!class /Script/FortniteGame.FortProjectileMovementInterface)
              (Projectile|SetHomingTimeUntilMaxTurnSpeed 0 _self 0.25 :!class /Script/FortniteGame.FortProjectileMovementInterface)
              (Utilities|IsValid _homingtarget
                (:Is%20Valid
                  (Projectile|SetHomingTarget 0 self _homingtarget 0 :!class /Script/FortniteGame.FortProjectileMovementInterface)
                  (Projectile|SetHomingTargetOffset self self (- (Collision|GetActorBounds _homingtarget true false) (Transformation|GetActorLocation _homingtarget)) :!class /Script/FortniteGame.FortProjectileMovementInterface))))))))))

(fn StartInAirAudio
  (bind _returnvalue (Audio|SpawnSoundAttached "/MotherGameplay/Items/GravyGoblin/Assets/Audio/Launcher_Gravy_Goblin_Prj_Disc_InAir_Loop_Cue.Launcher_Gravy_Goblin_Prj_Disc_InAir_Loop_Cue" (Variables|Default|GetRing1)))
  (Variables|Default|SetInAirAudio _returnvalue))

(fn StopInAirAudio
  (params (Play%20Loop%20Stop :type Boolean))
  (bind _returnvalue (CallFunction|GetOwningPlayer))
  (if (not (Variables|Default|GetStoppedAudio))
    (Audio|Components|Audio|FadeOut (Variables|Default|GetInAirAudio) 0.1 0.0 "Linear" :!class /Script/Engine.AudioComponent)
    (if Play%20Loop%20Stop
      (Utilities|IsValid _returnvalue
        (:Is%20Valid
          (Audio|SpawnSoundAttached "/MotherGameplay/Items/GravyGoblin/Assets/Audio/Launcher_Gravy_Goblin_Prj_Disc_InAir_Loop_Stop_Cue.Launcher_Gravy_Goblin_Prj_Disc_InAir_Loop_Stop_Cue" (Variables|/Script/Engine.Actor|GetRootComponent _returnvalue)))
        (:Is%20Not%20Valid
          _))
      (Variables|Default|SetStoppedAudio true)
      (else
        _))))

(fn UpdateInAirAudio
  (if (not (Variables|Default|GetStoppedAudio))
    (Audio|Parameter|SetFloatParameter (Variables|Default|GetInAirAudio) "Distance" (Math|Float|MapRangeClamped (Class|AbilityFunctions|GetDistanceBetweenActors (CallFunction|GetOwningPlayer) self) 0.0 2e+03 0.0 1.0) :!class /Script/AudioExtensions.AudioParameterControllerInterface)))

(fn ConstructionScript)
