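; Holder actor for the Gravy Goblin gravity gun, which carries a grabbed physics object, vehicle, projectile or pickup until it is thrown or dropped.
; Initialization tags and attaches the held object, spawning dummy stand-ins for projectiles and pickups.
; Launch logic branches by object type, respawning projectiles on the server, then a timeline retracts the ring to the weapon.
; ADS and fire are mirrored to non-local clients; toilet and tractor throws report quest stats.
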(event Custom|ADSPressedRemoteClients
  ;[0 0 0] ADS pressed
  (if (not (Pawn|IsLocallyControlled (Utilities|Casting|CastToFortPawn (Variables|FortGravityGunHolder|GetOwningPlayer))))
    (Variables|FortGravityGunHolder|SetbAimingDownSights true)
    (CallFunction|TurnonADSEffects)
    (bind _asplayer_pawn_athena (Utilities|GetPlayerPawnSafe)
      (:then
        _)
      (:Null
        _))
    (bind _returnvalue (Game|SpawnActorfromClass "/Game/Athena/Items/Consumables/Grenade/BP_ProjectileTrajectory_Athena.BP_ProjectileTrajectory_Athena_C" (Transformation|GetActorTransform) "AlwaysSpawn" "MultiplyWithRoot" 0 _asplayer_pawn_athena))
    (Variables|UpdateProjectileTrajectory|SetTrajectorIndicator _returnvalue)))

(event Custom|ADSPressed
  ;[0 0 0] ADS pressed
  (Networking|SwitchHasAuthority
    (:Authority
      (CallFunction|ADSPressedRemoteClients))
    (:Remote
      _))
  (Variables|FortGravityGunHolder|SetbAimingDownSights true)
  (CallFunction|TurnonADSEffects)
  (bind _asplayer_pawn_athena (Utilities|GetPlayerPawnSafe)
    (:then
      _)
    (:Null
      _))
  (bind _returnvalue (Game|SpawnActorfromClass "/Game/Athena/Items/Consumables/Grenade/BP_ProjectileTrajectory_Athena.BP_ProjectileTrajectory_Athena_C" (Transformation|GetActorTransform) "AlwaysSpawn" "MultiplyWithRoot" 0 _asplayer_pawn_athena))
  (Variables|UpdateProjectileTrajectory|SetTrajectorIndicator _returnvalue))

(event Custom|Fired
  (bind _owningplayer (Variables|FortGravityGunHolder|GetOwningPlayer))
  ;[1 1 1] OnDeath
  (Utilities|IsValid _owningplayer
    (:Is%20Valid
      (GameplayEffects|RemoveActiveEffectswithGrantedTags (Ability|GetFortAbilitySystemComponent _owningplayer) "(GameplayTags=((TagName=\"Granted.Athena.GravyGoblin.ProjectileInUse\")))")
      (Utilities|Array|RemoveItem (Variables|Actor|GetTags (Variables|FortGravityGunHolder|GetHeldObject)) "GrabbedByGravyGoblin")
      (Networking|SwitchHasAuthority
        (:Authority
          (CallFunction|TurnBackonHoldingPawnCollisionMulticast))))))

(event Custom|ADSReleasedRemoteClients
  ;[0 0 0] ADS released
  (if (not (Pawn|IsLocallyControlled (Utilities|Casting|CastToFortPawn (Variables|FortGravityGunHolder|GetOwningPlayer))))
    (Variables|FortGravityGunHolder|SetbAimingDownSights false)
    (CallFunction|TurnOffADSEffects)
    (Actor|DestroyActor (Variables|UpdateProjectileTrajectory|GetTrajectorIndicator))))

(event EventDestroyed
  (bind _owningplayer (Variables|FortGravityGunHolder|GetOwningPlayer))
  ;[1 1 1] OnDeath
  (Utilities|Time|ClearandInvalidateTimerbyHandle (Variables|Default|GetConnectionCheckTimer))
  (Utilities|IsValid _owningplayer
    (:Is%20Valid
      (GameplayEffects|RemoveActiveEffectswithGrantedTags (Ability|GetFortAbilitySystemComponent _owningplayer) "(GameplayTags=((TagName=\"Granted.Athena.GravyGoblin.ProjectileInUse\")))")
      (Utilities|Array|RemoveItem (Variables|Actor|GetTags (Variables|FortGravityGunHolder|GetHeldObject)) "GrabbedByGravyGoblin")
      (Networking|SwitchHasAuthority
        (:Authority
          (CallFunction|TurnBackonHoldingPawnCollisionMulticast))))))

(event Custom|ADSReleased
  ;[0 0 0] ADS released
  (Networking|SwitchHasAuthority
    (:Authority
      (CallFunction|ADSReleasedRemoteClients))
    (:Remote
      _))
  (Variables|FortGravityGunHolder|SetbAimingDownSights false)
  (CallFunction|TurnOffADSEffects)
  (Actor|DestroyActor (Variables|UpdateProjectileTrajectory|GetTrajectorIndicator)))

(event Custom|RotatePressed
  (bind (_currentrotationaxis _currentrotationflip) (Utilities|CalculateCurrentRotationSettings))
  ;[1 1 1] Rotation
  (Math|Integer|IncrementInt (Variables|/MotherGameplay/Items/GravyGoblin/B_GravyGoblin_Athena.B_GravyGoblin_Athena_C|GetCurrentRotations (Variables|Default|GetWeapon))
    (:%20%20
      (Variables|UpdateRotation|SetCurrentRotationAxisIndex _currentrotationaxis)
      (Variables|Default|SetFlipRotation _currentrotationflip))))

(event Custom|FirePressedRemoteClients
  (bind _heldobject (Variables|FortGravityGunHolder|GetHeldObject))
  (bind _returnvalue (Actor|GetComponentbyClass _heldobject "/Script/FortniteGame.FortPhysicsObjectComponent"))
  (bind _cached_endingscale (Variables|Default|GetCached_EndingScale))
  (bind _weapon (Variables|Default|GetWeapon))
  (bind _ns_gravygoblin_beam1 (Variables|/MotherGameplay/Items/GravyGoblin/B_GravyGoblin_Athena.B_GravyGoblin_Athena_C|GetNS_GravyGoblin_Beam1 _weapon))
  (bind _ns_gravitygoblin_holder (Variables|Default|GetNS_GravityGoblin_Holder))
  (bind _dummyobject (Variables|Default|GetDummyObject))
  (bind _heldobject_2 (Variables|FortGravityGunHolder|GetHeldObject))
  ;[1 1 1] LaunchObject
  (if (not (Pawn|IsLocallyControlled (Utilities|Casting|CastToFortPawn (Variables|FortGravityGunHolder|GetOwningPlayer))))
    (Utilities|CheckifConnectiontoObjectIsValid
      (:Blocked
        (Variables|Default|SetDropped (Class|AthenaFunctionLibrary|RowtoBool (Variables|Default|GetDontLaunchIfConnectionCutOff))))
      (:Clear
        _))
    (Utilities|FlowControl|DoOnce false
      (:Completed
        (Utilities|FlowControl|Sequence
          (:then_0
            (Utilities|Array|RemoveItem (Variables|Actor|GetTags _heldobject_2) "GrabbedByGravyGoblin")
            (Actor|DestroyActor (Variables|UpdateProjectileTrajectory|GetTrajectorIndicator))
            (Transformation|DetachFromActor "KeepWorld" "KeepWorld" "KeepWorld")
            (GravityGunHolderMethods|HandleThrowHeldObject)
            (CallFunction|QuestToiletorTractor _heldobject_2)
            (switch Utilities|FlowControl|Switch|SwitchonEGravityGunHolderObjectType (Variables|FortGravityGunHolder|GetObjectTypeHeld)
              (:PhysicsObject
                (bind (_outthrowspeed _outvelocity) (CallFunction|CalculateLaunchVelocity (Variables|Default|GetDropped) :!label _label))
                (FortPhysicsSimulationLibrary|SetPhysicsLinearVelocity (Variables|FortGravityGunHolder|GetHeldObject) _outvelocity false "None" 0 "(MaxSimSizeAllowed=Large,bRejectEmbeddedInLandscape=False)")
                (Utilities|IsValid _returnvalue
                  (:Is%20Valid
                    (Variables|/Script/FortniteGame.FortPhysicsObjectComponent|SetCustomGravityMultiplier 1.0 _returnvalue))))
              (:Vehicle
                (bind (_outthrowspeed_2 _outvelocity_2) (CallFunction|CalculateLaunchVelocity (Variables|Default|GetDropped)))
                (Audio|BroadcastAbilityImpactatLocation 0 (Variables|FortGravityGunHolder|GetHeldObject) _outvelocity_2 777.0)
                (GameplayEffects|ApplyGameplayEffectToTarget :self (Ability|GetFortAbilitySystemComponent (Variables|FortGravityGunHolder|GetOwningPlayer)) :GameplayEffectClass "/MotherGameplay/Items/GravyGoblin/GE_GravyGoblin_VehicleDamageOverride.GE_GravyGoblin_VehicleDamageOverride_C" :Target (Ability|GetFortAbilitySystemComponent (Variables|FortGravityGunHolder|GetHeldObject)) :Level 0.0) ; Damage override GE
              )
              (:Projectile
                (Actor|DestroyActor (Variables|FortGravityGunHolder|GetHeldObject))
                (Networking|SwitchHasAuthority
                  (:Authority
                    (bind _returnvalue_2 (Projectile|SpawnProjectile (Variables|FortGravityGunHolder|GetHeldObjectProjectileClass) (Variables|FortGravityGunHolder|GetOwningPlayer) (Transformation|GetActorLocation) (Camera|GetCameraRotation (Variables|Default|GetPlayerCameraManager)) (Camera|GetCameraRotation (Variables|Default|GetPlayerCameraManager)) (Variables|Default|GetEffect%20Container%20Spec%20to%20Apply%20on%20Hit) (Variables|Default|GetEffect%20Container%20Spec%20to%20Apply%20on%20Explode) true 0 0.0 0.0 1.0))
                    (bind _aspawn (Utilities|Casting|CastToPawn (Variables|FortGravityGunHolder|GetOwningPlayer)))
                    (Delivery|ChangeInstigator _returnvalue_2 _aspawn)
                    (Utilities|AddAdditionalExplosionEffects _returnvalue_2))
                  (:Remote
                    _))
                (Actor|DestroyActor _dummyobject)
                (Rendering|SetActorHiddenInGame _dummyobject true))
              (:PickUp
                (Utilities|FlowControl|Sequence
                  (:then_0
                    (bind _asbga_pickup_dummy_gravy_goblin (Utilities|Casting|CastToBGA_Pickup_Dummy_GravyGoblin (Variables|Default|GetDummyObject)))
                    (Class|BGAPickupDummyGravyGoblin|Throw _asbga_pickup_dummy_gravy_goblin))
                  (:then_1
                    (goto _label)))))
            (if (not (Variables|Default|GetDropped))
              (GameplayEffects|ApplyGameplayEffectToTarget :self (Ability|GetFortAbilitySystemComponent (Variables|FortGravityGunHolder|GetOwningPlayer)) :GameplayEffectClass "/MotherGameplay/Items/GravyGoblin/GE_ExecuteGC_FireObject.GE_ExecuteGC_FireObject_C" :Target (Ability|GetFortAbilitySystemComponent self) :Level 0.0) ; TODO figure out why this version isn't working
            )
            (bind _returnvalue_3 (Niagara|SpawnSystematLocation "/MotherGameplay/Items/GravyGoblin/Effects/NS_GravyGoblin_Hold_Launch.NS_GravyGoblin_Hold_Launch" (Transformation|GetWorldLocation (Variables|FortGravityGunHolder|GetRingPivotComponent)) (Transformation|GetWorldRotation (Variables|FortGravityGunHolder|GetRingPivotComponent))))
            (Effects|Components|ParticleSystem|SetFloatParameter _returnvalue_3 "RingSize" (Variables|Default|GetCached_EndingScale))
            (Niagara|SpawnSystemAttached "/MotherGameplay/Items/GravyGoblin/Effects/NS_GravyGoblin_MuzzleFlash_Throw.NS_GravyGoblin_MuzzleFlash_Throw" (Variables|/MotherGameplay/Items/GravyGoblin/B_GravyGoblin_Athena.B_GravyGoblin_Athena_C|GetNS_GravyGoblin_MuzzleFlash (Variables|Default|GetWeapon)) "None" "0, 0, 0" "0, 0, 0" "SnapToTarget")
            (Effects|Components|ParticleSystem|SetFloatParameter (Variables|Default|GetNS_GravityGoblin_Holder) "LaunchBrightness" 0.0)
            (Variables|FortGravityGunHolder|SetCurrentState "Detached")
            (Audio|Components|Audio|FadeOut (Variables|Default|GetMovementAudio) 1.0 0.0 "Linear" :!class /Script/Engine.AudioComponent)
            (Audio|SpawnSoundAttached (Variables|Default|GetObjectLaunchSound) (Variables|Transformation|GetRootComponent (Variables|FortGravityGunHolder|GetHeldObject)))
            (SoundIndicator|ShowSoundIndicator (Variables|/Script/FortniteGame.FortWeapon|GetSoundIndicatorComponent (Variables|Default|GetWeapon)) (Transformation|GetActorLocation (Variables|Default|GetWeapon)) (Variables|Default|GetObjectLaunchSound) 1.0 0)
            (Variables|Default|SetCachedStartRotation (Transformation|GetWorldRotation (Variables|FortGravityGunHolder|GetRingPivotComponent)))
            (bind (_direction _scale _goingback) (Timeline|Timeline_0 PlayFromStart)
              (:Update
                (Effects|Components|ParticleSystem|SetFloatParameter (Variables|/MotherGameplay/Items/GravyGoblin/B_GravyGoblin_Athena.B_GravyGoblin_Athena_C|GetNS_GravyGoblin_Beam1 (Variables|Default|GetWeapon)) "BeamBrightness" (Class|AthenaFunctionLibrary|UnclampedLerp 0.0 1.0 _scale))
                (Effects|Components|ParticleSystem|SetFloatParameter _ns_gravitygoblin_holder "RingSize" (* _cached_endingscale (Class|AthenaFunctionLibrary|UnclampedLerp 0.0 1.0 _scale)))
                (Effects|Components|ParticleSystem|SetFloatParameter _ns_gravitygoblin_holder "NewScale" (* _cached_endingscale (Class|AthenaFunctionLibrary|UnclampedLerp 0.0 1.0 _scale)))
                (Utilities|SetRingSize (* _cached_endingscale (Class|AthenaFunctionLibrary|UnclampedLerp 0.0 1.0 _scale)))
                (Transformation|SetWorldLocation (Variables|FortGravityGunHolder|GetRingPivotComponent) (Math|Vector|Lerp(Vector) (Transformation|GetActorLocation) (Transformation|GetWorldLocation (Variables|/MotherGameplay/Items/GravyGoblin/B_GravyGoblin_Athena.B_GravyGoblin_Athena_C|GetRing (Variables|Default|GetWeapon))) _goingback) false true)
                (Transformation|SetWorldRotation (Variables|FortGravityGunHolder|GetRingPivotComponent) (Math|Rotator|Lerp(Rotator) (Variables|Default|GetCachedStartRotation) "90.000000, 0, 0" _goingback true) false false)
                (Transformation|SetWorldScale3D (Variables|FortGravityGunHolder|GetRingPivotComponent) (Math|Vector|Lerp(Vector) "1.000000,1.000000,1.000000" "0.666000,0.666000,0.666000" _goingback)))
              (:Finished
                (Effects|Components|ParticleSystem|SetFloatParameter _ns_gravygoblin_beam1 "OpacityMult" 0.0)
                (Effects|Components|ParticleSystem|SetFloatParameter _ns_gravygoblin_beam1 "BeamBrightness" 0.0)
                (GameplayEffects|RemoveActiveEffectswithGrantedTags (Ability|GetFortAbilitySystemComponent (Variables|FortGravityGunHolder|GetOwningPlayer)) "(GameplayTags=((TagName=\"Granted.Athena.GravyGoblin.HoldingObject\")))")
                (Variables|/MotherGameplay/Items/GravyGoblin/B_GravyGoblin_Athena.B_GravyGoblin_Athena_C|SetHolder 0 (Variables|Default|GetWeapon))
                (Actor|DestroyActor)
                (Rendering|SetActorHiddenInGame true))
              (:ReshowRing
                (CallFunction|Fired))))
          (:then_1
            (Class|BGravyGoblinAthena|SetHoldingState (Variables|Default|GetWeapon) false)
            (bind _returnvalue_4 (GameplayEventRouter|GetEventRouter "/Script/FortniteGame.FortPlayspace" (Variables|Actor|GetInstigator)))
            (GameplayEventRouter|BroadcastEvent _returnvalue_4 (Utilities|Struct|MakeGravyGoblinRotatableEvent false) 0))
          (:then_2
            (Networking|SwitchHasAuthority
              (:Authority
                (CallFunction|TurnBackonHoldingPawnCollisionMulticast)))))))))

(event Custom|FirePressed
  (bind _heldobject (Variables|FortGravityGunHolder|GetHeldObject))
  (bind _returnvalue (Actor|GetComponentbyClass _heldobject "/Script/FortniteGame.FortPhysicsObjectComponent"))
  (bind _cached_endingscale (Variables|Default|GetCached_EndingScale))
  (bind _weapon (Variables|Default|GetWeapon))
  (bind _ns_gravygoblin_beam1 (Variables|/MotherGameplay/Items/GravyGoblin/B_GravyGoblin_Athena.B_GravyGoblin_Athena_C|GetNS_GravyGoblin_Beam1 _weapon))
  (bind _ns_gravitygoblin_holder (Variables|Default|GetNS_GravityGoblin_Holder))
  (bind _dummyobject (Variables|Default|GetDummyObject))
  (bind _heldobject_2 (Variables|FortGravityGunHolder|GetHeldObject))
  ;[1 1 1] LaunchObject
  (Networking|SwitchHasAuthority
    (:Authority
      (CallFunction|FirePressedRemoteClients))
    (:Remote
      _))
  (Utilities|CheckifConnectiontoObjectIsValid
    (:Blocked
      (Variables|Default|SetDropped (Class|AthenaFunctionLibrary|RowtoBool (Variables|Default|GetDontLaunchIfConnectionCutOff))))
    (:Clear
      _))
  (Utilities|FlowControl|DoOnce false
    (:Completed
      (Utilities|FlowControl|Sequence
        (:then_0
          (Utilities|Array|RemoveItem (Variables|Actor|GetTags _heldobject_2) "GrabbedByGravyGoblin")
          (Actor|DestroyActor (Variables|UpdateProjectileTrajectory|GetTrajectorIndicator))
          (Transformation|DetachFromActor "KeepWorld" "KeepWorld" "KeepWorld")
          (GravityGunHolderMethods|HandleThrowHeldObject)
          (CallFunction|QuestToiletorTractor _heldobject_2)
          (switch Utilities|FlowControl|Switch|SwitchonEGravityGunHolderObjectType (Variables|FortGravityGunHolder|GetObjectTypeHeld)
            (:PhysicsObject
              (bind (_outthrowspeed _outvelocity) (CallFunction|CalculateLaunchVelocity (Variables|Default|GetDropped) :!label _label))
              (FortPhysicsSimulationLibrary|SetPhysicsLinearVelocity (Variables|FortGravityGunHolder|GetHeldObject) _outvelocity false "None" 0 "(MaxSimSizeAllowed=Large,bRejectEmbeddedInLandscape=False)")
              (Utilities|IsValid _returnvalue
                (:Is%20Valid
                  (Variables|/Script/FortniteGame.FortPhysicsObjectComponent|SetCustomGravityMultiplier 1.0 _returnvalue))))
            (:Vehicle
              (bind (_outthrowspeed_2 _outvelocity_2) (CallFunction|CalculateLaunchVelocity (Variables|Default|GetDropped)))
              (Audio|BroadcastAbilityImpactatLocation 0 (Variables|FortGravityGunHolder|GetHeldObject) _outvelocity_2 777.0)
              (GameplayEffects|ApplyGameplayEffectToTarget :self (Ability|GetFortAbilitySystemComponent (Variables|FortGravityGunHolder|GetOwningPlayer)) :GameplayEffectClass "/MotherGameplay/Items/GravyGoblin/GE_GravyGoblin_VehicleDamageOverride.GE_GravyGoblin_VehicleDamageOverride_C" :Target (Ability|GetFortAbilitySystemComponent (Variables|FortGravityGunHolder|GetHeldObject)) :Level 0.0) ; Damage override GE
            )
            (:Projectile
              (Actor|DestroyActor (Variables|FortGravityGunHolder|GetHeldObject))
              (Networking|SwitchHasAuthority
                (:Authority
                  (bind _returnvalue_2 (Projectile|SpawnProjectile (Variables|FortGravityGunHolder|GetHeldObjectProjectileClass) (Variables|FortGravityGunHolder|GetOwningPlayer) (Transformation|GetActorLocation) (Camera|GetCameraRotation (Variables|Default|GetPlayerCameraManager)) (Camera|GetCameraRotation (Variables|Default|GetPlayerCameraManager)) (Variables|Default|GetEffect%20Container%20Spec%20to%20Apply%20on%20Hit) (Variables|Default|GetEffect%20Container%20Spec%20to%20Apply%20on%20Explode) true 0 0.0 0.0 1.0))
                  (bind _aspawn (Utilities|Casting|CastToPawn (Variables|FortGravityGunHolder|GetOwningPlayer)))
                  (Delivery|ChangeInstigator _returnvalue_2 _aspawn)
                  (Utilities|AddAdditionalExplosionEffects _returnvalue_2))
                (:Remote
                  _))
              (Actor|DestroyActor _dummyobject)
              (Rendering|SetActorHiddenInGame _dummyobject true))
            (:PickUp
              (Utilities|FlowControl|Sequence
                (:then_0
                  (bind _asbga_pickup_dummy_gravy_goblin (Utilities|Casting|CastToBGA_Pickup_Dummy_GravyGoblin (Variables|Default|GetDummyObject)))
                  (Class|BGAPickupDummyGravyGoblin|Throw _asbga_pickup_dummy_gravy_goblin))
                (:then_1
                  (goto _label)))))
          (if (not (Variables|Default|GetDropped))
            (GameplayEffects|ApplyGameplayEffectToTarget :self (Ability|GetFortAbilitySystemComponent (Variables|FortGravityGunHolder|GetOwningPlayer)) :GameplayEffectClass "/MotherGameplay/Items/GravyGoblin/GE_ExecuteGC_FireObject.GE_ExecuteGC_FireObject_C" :Target (Ability|GetFortAbilitySystemComponent self) :Level 0.0) ; TODO figure out why this version isn't working
          )
          (bind _returnvalue_3 (Niagara|SpawnSystematLocation "/MotherGameplay/Items/GravyGoblin/Effects/NS_GravyGoblin_Hold_Launch.NS_GravyGoblin_Hold_Launch" (Transformation|GetWorldLocation (Variables|FortGravityGunHolder|GetRingPivotComponent)) (Transformation|GetWorldRotation (Variables|FortGravityGunHolder|GetRingPivotComponent))))
          (Effects|Components|ParticleSystem|SetFloatParameter _returnvalue_3 "RingSize" (Variables|Default|GetCached_EndingScale))
          (Niagara|SpawnSystemAttached "/MotherGameplay/Items/GravyGoblin/Effects/NS_GravyGoblin_MuzzleFlash_Throw.NS_GravyGoblin_MuzzleFlash_Throw" (Variables|/MotherGameplay/Items/GravyGoblin/B_GravyGoblin_Athena.B_GravyGoblin_Athena_C|GetNS_GravyGoblin_MuzzleFlash (Variables|Default|GetWeapon)) "None" "0, 0, 0" "0, 0, 0" "SnapToTarget")
          (Effects|Components|ParticleSystem|SetFloatParameter (Variables|Default|GetNS_GravityGoblin_Holder) "LaunchBrightness" 0.0)
          (Variables|FortGravityGunHolder|SetCurrentState "Detached")
          (Audio|Components|Audio|FadeOut (Variables|Default|GetMovementAudio) 1.0 0.0 "Linear" :!class /Script/Engine.AudioComponent)
          (Audio|SpawnSoundAttached (Variables|Default|GetObjectLaunchSound) (Variables|Transformation|GetRootComponent (Variables|FortGravityGunHolder|GetHeldObject)))
          (SoundIndicator|ShowSoundIndicator (Variables|/Script/FortniteGame.FortWeapon|GetSoundIndicatorComponent (Variables|Default|GetWeapon)) (Transformation|GetActorLocation (Variables|Default|GetWeapon)) (Variables|Default|GetObjectLaunchSound) 1.0 0)
          (Variables|Default|SetCachedStartRotation (Transformation|GetWorldRotation (Variables|FortGravityGunHolder|GetRingPivotComponent)))
          (Timeline|Timeline_0 PlayFromStart))
        (:then_1
          (Class|BGravyGoblinAthena|SetHoldingState (Variables|Default|GetWeapon) false)
          (bind _returnvalue_4 (GameplayEventRouter|GetEventRouter "/Script/FortniteGame.FortPlayspace" (Variables|Actor|GetInstigator)))
          (GameplayEventRouter|BroadcastEvent _returnvalue_4 (Utilities|Struct|MakeGravyGoblinRotatableEvent false) 0))
        (:then_2
          (Networking|SwitchHasAuthority
            (:Authority
              (CallFunction|TurnBackonHoldingPawnCollisionMulticast))))))))

(event Custom|InitializeVariables
  (params (Held%20Object :type Actor%20Object%20Reference) (OwningPlayer :type Actor%20Object%20Reference) (ObjectTypeHeld :type EGravityGunHolderObjectType%20Enum :default "Invalid") (PrjClass :type Fort%20Projectile%20Base%20Class%20Reference))
  (bind _heldobject (Variables|FortGravityGunHolder|GetHeldObject))
  (bind _returnvalue (Actor|GetComponentbyClass _heldobject "/Script/FortniteGame.FortPhysicsObjectComponent"))
  (Utilities|FlowControl|Sequence
    (:then_0
      (Variables|FortGravityGunHolder|SetObjectTypeHeld ObjectTypeHeld))
    (:then_1
      (GravityGunHolderMethods|SetHeldObject Held%20Object)
      (bind _output_get (Variables|FortGravityGunHolder|SetOwningPlayer OwningPlayer))
      (Utilities|SetCameraManager _output_get)
      (Actor|SetOwner (Variables|FortGravityGunHolder|GetOwningPlayer))
      (Collision|SetActorEnableCollision false)
      (switch Utilities|FlowControl|Switch|SwitchonEGravityGunHolderObjectType (Variables|FortGravityGunHolder|GetObjectTypeHeld)
        (:PhysicsObject
          _)
        (:Vehicle
          (Transformation|SetRelativeRotation (Variables|FortGravityGunHolder|GetRingPivotComponent) "90.000000, 0, 0" false false :!label _label)
          (bind _returnvalue_2 (GameplayEventRouter|GetEventRouter "/Script/FortniteGame.FortPlayspace" (Variables|Actor|GetInstigator)))
          (GameplayEventRouter|BroadcastEvent _returnvalue_2 (Utilities|Struct|MakeGravyGoblinRotatableEvent false) 0))
        (:Projectile
          _)
        (:PickUp
          (goto _label)))
      (Utilities|IsValid _returnvalue
        (:Is%20Valid
          (Variables|/Script/FortniteGame.FortPhysicsObjectComponent|SetCustomGravityMultiplier 0.0 _returnvalue)
          (bind _returnvalue_3 (GameplayEventRouter|GetEventRouter "/Script/FortniteGame.FortPlayspace" (Variables|Actor|GetInstigator)))
          (GameplayEventRouter|BroadcastEvent _returnvalue_3 (Utilities|Struct|MakeGravyGoblinRotatableEvent true) 0))))
    (:then_2
      (bind _asplayer_pawn_athena (Utilities|GetPlayerPawnSafe))
      (bind _asb_gravy_goblin_athena (Utilities|Casting|CastToB_GravyGoblin_Athena (Variables|/Script/FortniteGame.FortPawn|GetCurrentWeapon _asplayer_pawn_athena)))
      (bind _output_get_2 (Variables|Default|SetWeapon _asb_gravy_goblin_athena))
      (Variables|/MotherGameplay/Items/GravyGoblin/B_GravyGoblin_Athena.B_GravyGoblin_Athena_C|SetHolder self _output_get_2)
      (Class|BGravyGoblinAthena|SetHoldingState _output_get_2 true))
    (:then_3
      (Variables|FortGravityGunHolder|SetHeldObjectProjectileClass PrjClass)
      (CallFunction|HandleGraborCaughtAudio))
    (:then_4
      (Variables|UpdateProjectileTrajectory|SetTrajectorySplineExtraIgnoreActors (Utilities|Array|MakeArray (Variables|FortGravityGunHolder|GetHeldObject)))
      (Effects|Components|ParticleSystem|SetFloatParameter (Variables|/MotherGameplay/Items/GravyGoblin/B_GravyGoblin_Athena.B_GravyGoblin_Athena_C|GetNS_GravyGoblin_Beam1 (Variables|Default|GetWeapon)) "ADS" 0.0))
    (:then_5
      (Variables|FortGravityGunHolder|SetCachedLayeredAudioComponent (Variables|Default|GetMovementAudio)))))

(event Custom|DropObject
  (bind _heldobject (Variables|FortGravityGunHolder|GetHeldObject))
  (bind _returnvalue (Actor|GetComponentbyClass _heldobject "/Script/FortniteGame.FortPhysicsObjectComponent"))
  (bind _cached_endingscale (Variables|Default|GetCached_EndingScale))
  (bind _weapon (Variables|Default|GetWeapon))
  (bind _ns_gravygoblin_beam1 (Variables|/MotherGameplay/Items/GravyGoblin/B_GravyGoblin_Athena.B_GravyGoblin_Athena_C|GetNS_GravyGoblin_Beam1 _weapon))
  (bind _ns_gravitygoblin_holder (Variables|Default|GetNS_GravityGoblin_Holder))
  (bind _dummyobject (Variables|Default|GetDummyObject))
  (bind _heldobject_2 (Variables|FortGravityGunHolder|GetHeldObject))
  ;[1 1 1] LaunchObject
  (Variables|Default|SetDropped true)
  (Utilities|FlowControl|DoOnce false
    (:Completed
      (Utilities|FlowControl|Sequence
        (:then_0
          (Utilities|Array|RemoveItem (Variables|Actor|GetTags _heldobject_2) "GrabbedByGravyGoblin")
          (Actor|DestroyActor (Variables|UpdateProjectileTrajectory|GetTrajectorIndicator))
          (Transformation|DetachFromActor "KeepWorld" "KeepWorld" "KeepWorld")
          (GravityGunHolderMethods|HandleThrowHeldObject)
          (CallFunction|QuestToiletorTractor _heldobject_2)
          (switch Utilities|FlowControl|Switch|SwitchonEGravityGunHolderObjectType (Variables|FortGravityGunHolder|GetObjectTypeHeld)
            (:PhysicsObject
              (bind (_outthrowspeed _outvelocity) (CallFunction|CalculateLaunchVelocity (Variables|Default|GetDropped) :!label _label))
              (FortPhysicsSimulationLibrary|SetPhysicsLinearVelocity (Variables|FortGravityGunHolder|GetHeldObject) _outvelocity false "None" 0 "(MaxSimSizeAllowed=Large,bRejectEmbeddedInLandscape=False)")
              (Utilities|IsValid _returnvalue
                (:Is%20Valid
                  (Variables|/Script/FortniteGame.FortPhysicsObjectComponent|SetCustomGravityMultiplier 1.0 _returnvalue))))
            (:Vehicle
              (bind (_outthrowspeed_2 _outvelocity_2) (CallFunction|CalculateLaunchVelocity (Variables|Default|GetDropped)))
              (Audio|BroadcastAbilityImpactatLocation 0 (Variables|FortGravityGunHolder|GetHeldObject) _outvelocity_2 777.0)
              (GameplayEffects|ApplyGameplayEffectToTarget :self (Ability|GetFortAbilitySystemComponent (Variables|FortGravityGunHolder|GetOwningPlayer)) :GameplayEffectClass "/MotherGameplay/Items/GravyGoblin/GE_GravyGoblin_VehicleDamageOverride.GE_GravyGoblin_VehicleDamageOverride_C" :Target (Ability|GetFortAbilitySystemComponent (Variables|FortGravityGunHolder|GetHeldObject)) :Level 0.0) ; Damage override GE
            )
            (:Projectile
              (Actor|DestroyActor (Variables|FortGravityGunHolder|GetHeldObject))
              (Networking|SwitchHasAuthority
                (:Authority
                  (bind _returnvalue_2 (Projectile|SpawnProjectile (Variables|FortGravityGunHolder|GetHeldObjectProjectileClass) (Variables|FortGravityGunHolder|GetOwningPlayer) (Transformation|GetActorLocation) (Camera|GetCameraRotation (Variables|Default|GetPlayerCameraManager)) (Camera|GetCameraRotation (Variables|Default|GetPlayerCameraManager)) (Variables|Default|GetEffect%20Container%20Spec%20to%20Apply%20on%20Hit) (Variables|Default|GetEffect%20Container%20Spec%20to%20Apply%20on%20Explode) true 0 0.0 0.0 1.0))
                  (bind _aspawn (Utilities|Casting|CastToPawn (Variables|FortGravityGunHolder|GetOwningPlayer)))
                  (Delivery|ChangeInstigator _returnvalue_2 _aspawn)
                  (Utilities|AddAdditionalExplosionEffects _returnvalue_2))
                (:Remote
                  _))
              (Actor|DestroyActor _dummyobject)
              (Rendering|SetActorHiddenInGame _dummyobject true))
            (:PickUp
              (Utilities|FlowControl|Sequence
                (:then_0
                  (bind _asbga_pickup_dummy_gravy_goblin (Utilities|Casting|CastToBGA_Pickup_Dummy_GravyGoblin (Variables|Default|GetDummyObject)))
                  (Class|BGAPickupDummyGravyGoblin|Throw _asbga_pickup_dummy_gravy_goblin))
                (:then_1
                  (goto _label)))))
          (if (not (Variables|Default|GetDropped))
            (GameplayEffects|ApplyGameplayEffectToTarget :self (Ability|GetFortAbilitySystemComponent (Variables|FortGravityGunHolder|GetOwningPlayer)) :GameplayEffectClass "/MotherGameplay/Items/GravyGoblin/GE_ExecuteGC_FireObject.GE_ExecuteGC_FireObject_C" :Target (Ability|GetFortAbilitySystemComponent self) :Level 0.0) ; TODO figure out why this version isn't working
          )
          (bind _returnvalue_3 (Niagara|SpawnSystematLocation "/MotherGameplay/Items/GravyGoblin/Effects/NS_GravyGoblin_Hold_Launch.NS_GravyGoblin_Hold_Launch" (Transformation|GetWorldLocation (Variables|FortGravityGunHolder|GetRingPivotComponent)) (Transformation|GetWorldRotation (Variables|FortGravityGunHolder|GetRingPivotComponent))))
          (Effects|Components|ParticleSystem|SetFloatParameter _returnvalue_3 "RingSize" (Variables|Default|GetCached_EndingScale))
          (Niagara|SpawnSystemAttached "/MotherGameplay/Items/GravyGoblin/Effects/NS_GravyGoblin_MuzzleFlash_Throw.NS_GravyGoblin_MuzzleFlash_Throw" (Variables|/MotherGameplay/Items/GravyGoblin/B_GravyGoblin_Athena.B_GravyGoblin_Athena_C|GetNS_GravyGoblin_MuzzleFlash (Variables|Default|GetWeapon)) "None" "0, 0, 0" "0, 0, 0" "SnapToTarget")
          (Effects|Components|ParticleSystem|SetFloatParameter (Variables|Default|GetNS_GravityGoblin_Holder) "LaunchBrightness" 0.0)
          (Variables|FortGravityGunHolder|SetCurrentState "Detached")
          (Audio|Components|Audio|FadeOut (Variables|Default|GetMovementAudio) 1.0 0.0 "Linear" :!class /Script/Engine.AudioComponent)
          (Audio|SpawnSoundAttached (Variables|Default|GetObjectLaunchSound) (Variables|Transformation|GetRootComponent (Variables|FortGravityGunHolder|GetHeldObject)))
          (SoundIndicator|ShowSoundIndicator (Variables|/Script/FortniteGame.FortWeapon|GetSoundIndicatorComponent (Variables|Default|GetWeapon)) (Transformation|GetActorLocation (Variables|Default|GetWeapon)) (Variables|Default|GetObjectLaunchSound) 1.0 0)
          (Variables|Default|SetCachedStartRotation (Transformation|GetWorldRotation (Variables|FortGravityGunHolder|GetRingPivotComponent)))
          (Timeline|Timeline_0 PlayFromStart))
        (:then_1
          (Class|BGravyGoblinAthena|SetHoldingState (Variables|Default|GetWeapon) false)
          (bind _returnvalue_4 (GameplayEventRouter|GetEventRouter "/Script/FortniteGame.FortPlayspace" (Variables|Actor|GetInstigator)))
          (GameplayEventRouter|BroadcastEvent _returnvalue_4 (Utilities|Struct|MakeGravyGoblinRotatableEvent false) 0))
        (:then_2
          (Networking|SwitchHasAuthority
            (:Authority
              (CallFunction|TurnBackonHoldingPawnCollisionMulticast))))))))

(event Custom|FireReleased)

(event Custom|Detach
  (bind _cached_endingscale (Variables|Default|GetCached_EndingScale))
  (bind _weapon (Variables|Default|GetWeapon))
  (bind _ns_gravygoblin_beam1 (Variables|/MotherGameplay/Items/GravyGoblin/B_GravyGoblin_Athena.B_GravyGoblin_Athena_C|GetNS_GravyGoblin_Beam1 _weapon))
  (bind _ns_gravitygoblin_holder (Variables|Default|GetNS_GravityGoblin_Holder))
  ;[1 1 1] LaunchObject
  (Transformation|DetachFromActor (Variables|FortGravityGunHolder|GetHeldObject) "KeepWorld" "KeepWorld" "KeepWorld")
  (Variables|FortGravityGunHolder|SetCurrentState "Detached")
  (Audio|Components|Audio|FadeOut (Variables|Default|GetMovementAudio) 1.0 0.0 "Linear" :!class /Script/Engine.AudioComponent)
  (Audio|SpawnSoundAttached (Variables|Default|GetObjectLaunchSound) (Variables|Transformation|GetRootComponent (Variables|FortGravityGunHolder|GetHeldObject)))
  (SoundIndicator|ShowSoundIndicator (Variables|/Script/FortniteGame.FortWeapon|GetSoundIndicatorComponent (Variables|Default|GetWeapon)) (Transformation|GetActorLocation (Variables|Default|GetWeapon)) (Variables|Default|GetObjectLaunchSound) 1.0 0)
  (Variables|Default|SetCachedStartRotation (Transformation|GetWorldRotation (Variables|FortGravityGunHolder|GetRingPivotComponent)))
  (Timeline|Timeline_0 PlayFromStart))

(event Custom|HeldObjectDied
  (params (DamagedActor :type Actor%20Object%20Reference) (Damage :type Float%20%28single-precision%29) (InstigatedBy :type Controller%20Object%20Reference) (DamageCauser :type Actor%20Object%20Reference) (HitLocation :type Vector) (FHitComponent :type Primitive%20Component%20Object%20Reference) (BoneName :type Name) (Momentum :type Vector))
  (Networking|SwitchHasAuthority
    (:Authority
      (CallFunction|DropObjectMulticast))))

(event Custom|DropObjectMulticast
  (CallFunction|DropObject))

(event Custom|Initialize2
  (params (HeldObject :type Actor%20Object%20Reference) (OwningPlayer :type Actor%20Object%20Reference) (ObjectType :type EGravityGunHolderObjectType%20Enum :default "Invalid") (ProjectileClass :type Fort%20Projectile%20Base%20Class%20Reference))
  (bind _heldobject (Variables|FortGravityGunHolder|GetHeldObject))
  (if (not (Variables|Default|GetInitialized))
    (Development|PrintString "INIT" true true "(R=0.000000,G=0.660000,B=1.000000,A=1.000000)" 15.0)
    (Variables|Default|SetInitialized true)
    (CallFunction|InitializeVariables HeldObject OwningPlayer ObjectType ProjectileClass)
    (Utilities|Array|AddUnique (Variables|Actor|GetTags (Variables|FortGravityGunHolder|GetHeldObject)) "GrabbedByGravyGoblin")
    (CallFunction|InitializeRingSizeandRotation)
    (switch Utilities|FlowControl|Switch|SwitchonEGravityGunHolderObjectType (Variables|FortGravityGunHolder|GetObjectTypeHeld)
      (:PhysicsObject
        _)
      (:Vehicle
        _)
      (:Projectile
        (Variables|Default|SetOriginalProjectile _heldobject)
        (Networking|SwitchHasAuthority :!label _label
          (:Authority
            (bind _returnvalue (Game|SpawnActorfromClass (select (Utilities|Enum|Equal(Enum) (Variables|FortGravityGunHolder|GetObjectTypeHeld) "PickUp") "/MotherGameplay/Items/GravyGoblin/BGA_Pickup_Dummy_GravyGoblin.BGA_Pickup_Dummy_GravyGoblin_C" "/MotherGameplay/Items/GravyGoblin/BGA_Prj_Dummy_GravyGoblin__.BGA_Prj_Dummy_GravyGoblin___C") (Transformation|GetActorTransform (Variables|FortGravityGunHolder|GetHeldObject)) "AlwaysSpawn" "SelectDefaultAtRuntime" 0 0))
            (bind _asbga_athena_physics_parent (Utilities|Casting|CastToBGA_Athena_Physics_Parent _returnvalue))
            (Variables|Default|SetDummyObject _asbga_athena_physics_parent))
          (:Remote
            (if (Variables|Default|GetSwapInDummyOnInit)
              (CallFunction|SwapinDummyObject)))))
      (:PickUp
        (Variables|Default|SetOriginalPickup _heldobject)
        (Utilities|FlowControl|Sequence
          (:then_0
            (Game|BindEventtoOnDestroyed (Variables|Default|GetOriginalPickup) (delegate Custom|PickUpDestroyed)) ; held object is the pickup at this point
          )
          (:then_1
            (goto _label)))))
    (CallFunction|Attach)
    (bind _asbuilding_actor (Utilities|Casting|CastToBuildingActor (Variables|FortGravityGunHolder|GetHeldObject))
      (:then
        (EventDispatchers|BindEventtoOnDied _asbuilding_actor (delegate Custom|HeldObjectDied) :!class /Script/FortniteGame.BuildingActor))
      (:CastFailed
        (Development|PrintString "ERROR in GravyGoblin Holder, tried to bind to non building actor" true true "(R=1.000000,G=0.000000,B=0.008842,A=1.000000)")))))

(event Custom|ToiletLaunchQuestEvent
  (Utilities|FlowControl|DoOnce false
    (:Completed
      (Quest|SendComplexCustomStatEvent (Quests|GetQuestManager (Fortnite|GetFortPlayerControllerfromActor (Variables|FortGravityGunHolder|GetOwningPlayer)) "Invalid") 0 "(GameplayTags=((TagName=\"Athena.Quests.GravyGoblin.ToiletLaunch\")))" "(GameplayTags=)" 1))))

(event Custom|Quest_ToiletOrTractor
  (params (Held%20Object :type Actor%20Object%20Reference))
  ;[1 1 1] Quest- check if launched object is a toilet or tractor
  (Utilities|Casting|CastToBathroom_ToiletWithTank01 Held%20Object
    (:then
      _)
    (:CastFailed
      (Utilities|Casting|CastToBathroom_Urinal01 Held%20Object
        (:then
          _)
        (:CastFailed
          (Utilities|Casting|CastToB_HidingProp_Portapotty Held%20Object
            (:then
              _)
            (:CastFailed
              (Utilities|Casting|CastToNeoTilted_Toilet_03 Held%20Object
                (:then
                  _)
                (:CastFailed
                  (Utilities|Casting|CastToBathroom_ToiletBowl01 Held%20Object
                    (:then
                      _)
                    (:CastFailed
                      (Utilities|Casting|CastToAthena_Prop_Bathroom_POI_FlushFactory_ToiletXL Held%20Object
                        (:then
                          _)
                        (:CastFailed
                          (Utilities|Casting|CastToAthena_Prop_Toilet_Factory_Toilet_Future Held%20Object
                            (:then
                              _)
                            (:CastFailed
                              (Utilities|Casting|CastToAthena_Prop_Bathroom_Toilet_01 Held%20Object
                                (:then
                                  _)
                                (:CastFailed
                                  (Utilities|Casting|CastToAthena_Prop_Bathroom_Toilet_02 Held%20Object
                                    (:then
                                      _)
                                    (:CastFailed
                                      (Utilities|Casting|CastToApollo_Suburban_ToiletBowl Held%20Object
                                        (:then
                                          _)
                                        (:CastFailed
                                          (Utilities|Casting|CastToApollo_Suburban_Toilet_04 Held%20Object
                                            (:then
                                              _)
                                            (:CastFailed
                                              (Utilities|Casting|CastToApollo_Farm_Tractor_01 Held%20Object)
                                              (Utilities|FlowControl|DoOnce false
                                                (:Completed
                                                  (CallFunction|TractorLaunchQuestEvent)))))))))))))))))))))))))
  (Utilities|FlowControl|DoOnce false
    (:Completed
      (CallFunction|ToiletLaunchQuestEvent))))

(event Custom|Initialize
  (params (HeldObject :type Actor%20Object%20Reference) (OwningPlayer :type Actor%20Object%20Reference) (ObjectType :type EGravityGunHolderObjectType%20Enum :default "Invalid") (ProjectileClass :type Fort%20Projectile%20Base%20Class%20Reference) (Effect%20Container%20Spec%20to%20Apply%20on%20Hit :type Fort%20Gameplay%20Effect%20Container%20Spec%20Structure) (Effect%20Container%20Spec%20to%20Apply%20on%20Explode :type Fort%20Gameplay%20Effect%20Container%20Spec%20Structure) (AdditionalExplodeEffects :type Array%20of%20Fort%20Gameplay%20Effect%20Container%20Spec%20Structures))
  (Networking|SwitchHasAuthority
    (:Authority
      (Variables|Default|SetEffect%20Container%20Spec%20to%20Apply%20on%20Hit Effect%20Container%20Spec%20to%20Apply%20on%20Hit)
      (Variables|Default|SetEffect%20Container%20Spec%20to%20Apply%20on%20Explode Effect%20Container%20Spec%20to%20Apply%20on%20Explode)
      (Variables|Default|SetAdditional%20Explode%20Effects AdditionalExplodeEffects)
      (Variables|SetHolderInfo_HeldObject_2_ACB32FEE4C7A2835090E81B8E298D87D (Utilities|Struct|MakeGravyGoblinHolderInfo HeldObject OwningPlayer ObjectType ProjectileClass)) ; for clients that are far away when it's initialized
    )
    (:Remote
      _))
  (CallFunction|Initialize2 HeldObject OwningPlayer ObjectType ProjectileClass) ; TODO connect this later
)

(event Custom|TractorLaunchQuestEvent
  (Utilities|FlowControl|DoOnce false
    (:Completed
      (Quest|SendComplexCustomStatEvent (Quests|GetQuestManager (Fortnite|GetFortPlayerControllerfromActor (Variables|FortGravityGunHolder|GetOwningPlayer)) "Invalid") 0 "(GameplayTags=((TagName=\"Athena.Quests.GravyGoblin.TractorLaunch\")))" "(GameplayTags=)" 1))))

(event Custom|PickUpDestroyed
  (params (DestroyedActor :type Actor%20Object%20Reference))
  (CallFunction|PickUpDestroyedServer)
  (Actor|DestroyActor (Variables|Default|GetDummyObject))
  (CallFunction|DropObject))

(event Custom|PickUpDestroyedServer
  (Actor|DestroyActor (Variables|Default|GetDummyObject))
  (CallFunction|DropObject))

(event Custom|Attach
  (bind _heldobject (Variables|FortGravityGunHolder|GetHeldObject))
  (bind _origin (Collision|GetActorBounds _heldobject true false))
  (bind _self self)
  (bind _originalprojectile (Variables|Default|GetOriginalProjectile))
  (Utilities|FlowControl|Sequence
    (:then_0
      (Transformation|AttachActorToActor _self _heldobject "None" "KeepWorld" "KeepWorld" "KeepWorld" true)
      (Transformation|SetActorLocation _self _origin false true)
      (CallFunction|GrowRing true)
      (bind _returnvalue (Niagara|SpawnSystematLocation "/MotherGameplay/Items/GravyGoblin/Effects/NS_GravyGoblin_Grab.NS_GravyGoblin_Grab" _origin))
      (Effects|Components|ParticleSystem|SetFloatParameter _returnvalue "RingSize" (Variables|Default|GetCached_EndingScale))
      (Game|BindEventtoOnDestroyed (Variables|FortGravityGunHolder|GetHeldObject) (delegate Custom|HeldObjectDestroyed))
      (Variables|Default|SetFullyInitialized true)
      (Utilities|IsValid _originalprojectile
        (:Is%20Valid
          (Transformation|SetActorLocation _originalprojectile (0.0 0.0 -99999.0) false true))))
    (:then_1
      (Ability|Tasks|WaitGameplayTagRemoveFromActor (Variables|FortGravityGunHolder|GetOwningPlayer) "(TagName=\"Granted.Athena.GravyGoblin.ProjectileInUse\")"
        (:then
          (if (Class|AthenaFunctionLibrary|RowtoBool (Variables|Default|GetDropObjectWhenConnectionCutOff))
            (bind _returnvalue_2 (Utilities|Time|SetTimerbyEvent (delegate Custom|CheckForConnectionCut) (ScalableFloat|GetValueatLevel (Variables|Default|GetDropObjectWhenConnectionCutOff_UpdateTime)) true))
            (Variables|Default|SetConnectionCheckTimer _returnvalue_2)))
        (:Removed
          (if (Class|AthenaFunctionLibrary|RowtoBool (Variables|Default|GetDropIfProjectileNoLongerInUse))
            (Networking|SwitchHasAuthority
              (:Authority
                (CallFunction|DropObjectMulticast))) ; Need to do this on the server because\nthis sometimes missfires on the client\nFORT-389787
          ))) ; failsafe
    )))

(event Custom|HeldObjectDestroyed
  (params (DestroyedActor :type Actor%20Object%20Reference))
  (CallFunction|DropObject))

(event Custom|TurnOnADSEffects
  ;[1 1 1] ADS effects
  (bind (_direction _timeline) (Timeline|TurnOnADSEffects_ PlayFromStart)
    (:Update
      (Effects|Components|ParticleSystem|SetFloatParameter (Variables|/MotherGameplay/Items/GravyGoblin/B_GravyGoblin_Athena.B_GravyGoblin_Athena_C|GetNS_GravyGoblin_Beam1 (Variables|Default|GetWeapon)) "ADS" _timeline))))

(event Custom|TurnOffADSEffects
  ;[1 1 1] ADS effects
  (Timeline|TurnOnADSEffects_ Reverse))

(event Custom|Grow%20Ring
  (params (WobbleObject :type Boolean))
  (bind _heldobject (Variables|FortGravityGunHolder|GetHeldObject))
  (bind _ns_gravitygoblin_holder (Variables|Default|GetNS_GravityGoblin_Holder))
  (bind _cached_endingscale (Variables|Default|GetCached_EndingScale))
  (bind _weapon (Variables|Default|GetWeapon))
  ;[1 1 1] Grow Ring
  (Variables|Default|SetCached_WobbleObject WobbleObject)
  (Effects|Components|ParticleSystem|SetFloatParameter (Variables|/MotherGameplay/Items/GravyGoblin/B_GravyGoblin_Athena.B_GravyGoblin_Athena_C|GetNS_GravyGoblin_Beam1 (Variables|Default|GetWeapon)) "InFlight" 0.0)
  (Variables|Default|SetCached_HeldObjectRotation (Transformation|GetActorRotation (Variables|FortGravityGunHolder|GetHeldObject)))
  (Variables|Default|SetCached_StartingScale 1.0)
  (bind (_direction _scale _wobble) (Timeline|GrowRing Play)
    (:Update
      (Effects|Components|ParticleSystem|SetFloatParameter _ns_gravitygoblin_holder "RingSize" (* _cached_endingscale (Class|AthenaFunctionLibrary|UnclampedLerp 0.0 1.0 _scale)))
      (Effects|Components|ParticleSystem|SetFloatParameter _ns_gravitygoblin_holder "NewScale" (* _cached_endingscale (Class|AthenaFunctionLibrary|UnclampedLerp 0.0 1.0 _scale)))
      (Utilities|SetRingSize (* _cached_endingscale (Class|AthenaFunctionLibrary|UnclampedLerp 0.0 1.0 _scale)))
      (if (Variables|Default|GetCached_WobbleObject)
        (switch Utilities|FlowControl|Switch|SwitchonEGravityGunHolderState (Variables|FortGravityGunHolder|GetCurrentState)
          (:Unrooting
            (if (Utilities|Enum|NotEqual(Enum) (Variables|FortGravityGunHolder|GetObjectTypeHeld) "Vehicle")
              (Transformation|SetActorRotation (Variables|FortGravityGunHolder|GetHeldObject) (Math|Rotator|CombineRotators (Variables|Default|GetCached_HeldObjectRotation) (Math|Rotator|MakeRotator (* _wobble 1e+01))) false))))))
    (:HasObject
      (Utilities|FlowControl|Sequence
        (:then_0
          (Variables|FortGravityGunHolder|SetCurrentState "Holding")
          (Variables|UpdateLocation|SetTimeWhenStartedHolding (Utilities|Time|GetGameTimeinSeconds))
          (Variables|Default|SetCurrentRightDistance (Variables|Default|GetHoldRightDistance))
          (Variables|Default|SetCurrentUpDistance (ScalableFloat|GetValueatLevel (Variables|UpdateLocation|GetHoldUpDistance)))
          (Variables|Default|SetCurrentForwardDistance (ScalableFloat|GetValueatLevel (Variables|UpdateLocation|GetHoldForwardDistance)))
          (switch Utilities|FlowControl|Switch|SwitchonEGravityGunHolderObjectType (Variables|FortGravityGunHolder|GetObjectTypeHeld)
            (:PhysicsObject
              (Transformation|SetActorRotation (Variables|FortGravityGunHolder|GetHeldObject) (Variables|Default|GetCached_HeldObjectRotation) false :!label _label)
              (FortPhysicsSimulationLibrary|WakeUp (Variables|FortGravityGunHolder|GetHeldObject) "(MaxSimSizeAllowed=Large,bRejectEmbeddedInLandscape=False)"))
            (:Vehicle
              _)
            (:Projectile
              (goto _label))
            (:PickUp
              (goto _label)))
          (FortPhysicsSimulationLibrary|SetImpactInstigatorforPhysicsObject (Variables|FortGravityGunHolder|GetHeldObject) (Variables|FortGravityGunHolder|GetOwningPlayer)))
        (:then_1
          (bind _asb_held_object_parent (Utilities|Casting|CastToB_HeldObject_Parent _heldobject)
            (:then
              (bind _returnvalue (Fortnite|HeldObjects|IsHeldObjectBeingHeld _asb_held_object_parent))
              (if _returnvalue
                (bind _returnvalue_2 (Fortnite|HeldObjects|GetHeldObjectComponentAttachedtoPlayer (Variables|/Game/Athena/Items/HeldObjects/B_HeldObject_Parent.B_HeldObject_Parent_C|GetHoldingPawn _asb_held_object_parent)))
                (HeldObjects|Behavior|DropHeldObject _returnvalue_2)))
            (:CastFailed
              (bind _asbga_held_object_physics_parent (Utilities|Casting|CastToBGA_HeldObject_Physics_Parent _heldobject))
              (HeldObjects|Behavior|DropHeldObject (Variables|/Game/Athena/Items/HeldObjects/HeldObjectVersion2/BGA_HeldObject_Physics_Parent.BGA_HeldObject_Physics_Parent_C|GetFortHeldObject _asbga_held_object_physics_parent)))))
        (:then_2
          (Utilities|FlowControl|Sequence
            (:then_0
              ;[1 1 1] Check for already ADS
              (bind _returnvalue_3 (GameplayTags|HasMatchingGameplayTag(Message) (Variables|FortGravityGunHolder|GetOwningPlayer) "(TagName=\"Gameplay.Status.IsTargeting\")"))
              (if _returnvalue_3
                (CallFunction|ADSPressed)))
            (:then_1
              ;[1 1 1] Check if weapon changed while we were setting up
              (bind _asplayer_pawn_athena (Utilities|GetPlayerPawnSafe))
              (if (not (or (Weapon|IsWeaponHolstered _asplayer_pawn_athena) (Utilities|NotEqual(Object) (Variables|/Script/FortniteGame.FortPawn|GetCurrentWeapon _asplayer_pawn_athena) _weapon)))
                (Utilities|FlowControl|Delay 1.5) ; failsafe, to fix FORT-386109
                (if (or (Weapon|IsWeaponHolstered _asplayer_pawn_athena) (Utilities|NotEqual(Object) (Variables|/Script/FortniteGame.FortPawn|GetCurrentWeapon _asplayer_pawn_athena) _weapon))
                  _))
              (CallFunction|DropObject))
            (:then_2
              ;[1 1 1] Apply HoldingObject GE
              (GameplayEffects|ApplyGameplayEffectToTarget (Ability|GetFortAbilitySystemComponent (Variables|FortGravityGunHolder|GetOwningPlayer)) "/MotherGameplay/Items/GravyGoblin/GE_HoldingObject_GravyGoblin.GE_HoldingObject_GravyGoblin_C" (Ability|GetFortAbilitySystemComponent (Variables|FortGravityGunHolder|GetOwningPlayer)) 0.0 (GameplayEffects|MakeFortEffectContext (Ability|GetFortAbilitySystemComponent (Variables|FortGravityGunHolder|GetOwningPlayer)) 0 self 0 0)))
            (:then_3
              (Networking|SwitchHasAuthority
                (:Authority
                  (CallFunction|IgnoreHoldingPawnCollisionMulticast))))))))))

(event Custom|CheckForConnectionCut
  (Utilities|CheckifConnectiontoObjectIsValid
    (:Blocked
      (CallFunction|DropObject))))

(event OnOneshotEvent(MovementAudio)
  (params (AudioComponent :type Audio%20Component%20Object%20Reference) (ParameterName :type Name :default "None") (OneshotIndex :type Integer))
  (if (Utilities|Name|Equal(Name) ParameterName "Velocity")
    (Audio|Parameter|SetIntegerParameter AudioComponent "Size" (CallFunction|GetPhysicsObjectSize (Variables|FortGravityGunHolder|GetHeldObject)) :!class /Script/AudioExtensions.AudioParameterControllerInterface)))

(event Custom|IgnoreHoldingPawnCollision_Multicast
  (CallFunction|IgnoreHoldingPawnCollision))

(event Custom|TurnBackOnHoldingPawnCollision_Multicast
  (Default|TurnBackonHoldingPawnCollision))

(event EventBeginPlay
  (bind _heldobject (Variables|FortGravityGunHolder|GetHeldObject))
  (|CallParentFunction)
  (Development|PrintString (Utilities|String|ToString(Float) (Actor|GetGameTimeSinceCreation)) true true "(R=1.000000,G=0.137311,B=0.000000,A=1.000000)" 12.0)
  (Utilities|IsValid _heldobject
    (:Is%20Valid
      (Utilities|Array|AddUnique (Variables|Actor|GetTags _heldobject) "GrabbedByGravyGoblin"))))

(fn GetHeldObjectBoundingBox
  (params (Actor :type Actor%20Object%20Reference))
  (bind _returnvalue (Actor|GetComponentsByClass Actor "/Script/Engine.StaticMeshComponent"))
  (bind _heldobject (Variables|FortGravityGunHolder|GetHeldObject))
  (bind (_origin _boxextent) (Collision|GetActorBounds _heldobject true false))
  (bind (_currentmax_x _currentmax_y _currentmax_z) (Math|Vector|BreakVector (Variables|GetCurrentMax_X)))
  (bind (_currentmin_x _currentmin_y _currentmin_z) (Math|Vector|BreakVector (Variables|GetCurrentMin_X)))
  (if (> (Utilities|Array|Length _returnvalue) 0)
    (for _array_element _returnvalue
      (if (and (Rendering|IsVisible _array_element) (not (Variables|/Script/Engine.SceneComponent|GetbHiddenInGame _array_element)))
        (Utilities|FlowControl|Sequence
          (:then_0
            (bind (_returnvalue_min _returnvalue_max) (Utilities|Struct|BreakBox (StaticMesh|GetBoundingBox (Variables|/Script/Engine.StaticMeshComponent|GetStaticMesh _array_element))))
            (bind (_returnvalue_min_x _returnvalue_min_y _returnvalue_min_z) (Math|Vector|BreakVector _returnvalue_min))
            (bind (_returnvalue_max_x _returnvalue_max_y _returnvalue_max_z) (Math|Vector|BreakVector _returnvalue_max))
            (Variables|SetCurrentMin_X (Math|Vector|MakeVector (Math|Float|Min(Float) _currentmin_x _returnvalue_min_x) (Math|Float|Min(Float) _currentmin_y _returnvalue_min_y) (Math|Float|Min(Float) _currentmin_z _returnvalue_min_z))))
          (:then_1
            (Variables|SetCurrentMax_X (Math|Vector|MakeVector (Math|Float|Max(Float) _returnvalue_max_x _currentmax_x) (Math|Float|Max(Float) _returnvalue_max_y _currentmax_y) (Math|Float|Max(Float) _returnvalue_max_z _currentmax_z)))))))
    (if (Math|Vector|Equal(Vector) (LocalVariables|GetCurrentMax) "0, 0, 0" 0.0001)
      (return _boxextent)
      (else
        (return (- (LocalVariables|GetCurrentMax) (LocalVariables|GetCurrentMin)))))
    (else
      (return _boxextent))))

(fn GetPhysicsObjectSize
  (params (PhysicsObject :type Actor%20Object%20Reference))
  (bind _returnvalue (FortPhysicsSimulationLibrary|DoesPhysicsObjectMeetSizeRequirement PhysicsObject "Small"))
  (if _returnvalue
    (return 0)
    (else
      (bind _returnvalue_2 (FortPhysicsSimulationLibrary|DoesPhysicsObjectMeetSizeRequirement PhysicsObject "Average"))
      (if _returnvalue_2
        (return 1)
        (else
          (return 2))))))

(fn HandleGrabOrCaughtAudio
  (bind _returnvalue (Audio|SpawnSoundAttached "/MotherGameplay/Items/GravyGoblin/Assets/Audio/Launcher_Gravy_Goblin_ObjectGrabbed_Start_Cue.Launcher_Gravy_Goblin_ObjectGrabbed_Start_Cue" (Variables|Default|GetMovementAudio)))
  (Audio|Parameter|SetIntegerParameter _returnvalue "Size" (CallFunction|GetPhysicsObjectSize (Variables|FortGravityGunHolder|GetHeldObject)) :!class /Script/AudioExtensions.AudioParameterControllerInterface)
  (bind _asfort_pawn (Utilities|Casting|CastToFortPawn (Variables|FortGravityGunHolder|GetOwningPlayer)))
  (Audio|SpawnSoundAttached (select (Pawn|IsLocallyControlled _asfort_pawn) "/MotherGameplay/Items/GravyGoblin/Assets/Audio/Launcher_Gravy_Goblin_Lift_Weapon_Lock_1P_Cue.Launcher_Gravy_Goblin_Lift_Weapon_Lock_1P_Cue" "/MotherGameplay/Items/GravyGoblin/Assets/Audio/Launcher_Gravy_Goblin_Lift_Weapon_Lock_Cue.Launcher_Gravy_Goblin_Lift_Weapon_Lock_Cue") (Variables|Transformation|GetRootComponent _asfort_pawn)))

(fn IgnoreHoldingPawnCollision
  (bind _heldobject (Variables|FortGravityGunHolder|GetHeldObject))
  (bind _returnvalue (Actor|GetComponentbyClass _heldobject "/Script/FortniteGame.FortPhysicsObjectComponent"))
  (if (Class|AthenaFunctionLibrary|RowtoBool (Variables|Default|GetIgnoreCollisionWithPlayerHolding))
    (Utilities|IsValid _heldobject
      (:Is%20Valid
        (Utilities|IsValid _returnvalue
          (:Is%20Valid
            (PhysicsObject|IncrementNeverSleep _returnvalue)
            (Variables|Default|SetHoldingObject true)
            (bind _asfort_pawn (Utilities|Casting|CastToFortPawn (Variables|FortGravityGunHolder|GetOwningPlayer)))
            (PhysicsObject|IgnorePawnforCollision _returnvalue _asfort_pawn true true)))))))

(fn InitializeRingSizeAndRotation
  (bind (_x _y _z) (Math|Rotator|GetAxes (Transformation|GetActorRotation (Variables|FortGravityGunHolder|GetHeldObject))))
  (bind _boundingboxextents (Class|AthenaFunctionLibrary|GetActorMeshExtents (Variables|FortGravityGunHolder|GetHeldObject)))
  (Variables|Default|SetHeldObjectMass (Math|Vector|VectorLength _boundingboxextents))
  (Variables|FortGravityGunHolder|SetCurrentState "Unrooting")
  (Variables|UpdateLocation|SetMaxAxisLength (Math|Float|Max(Float) (.x _boundingboxextents) (.y _boundingboxextents) (.z _boundingboxextents)))
  (Utilities|CaclulateandSetMaxSurfaceAxis (.x _boundingboxextents) (.y _boundingboxextents) (.z _boundingboxextents))
  (Variables|UpdateLocation|SetCurrentRingSizeInUnits 0.0)
  (Variables|Default|SetCached_EndingScale (* (Math|Float|Clamp(Float) (Math|Float|Min(Float) (.x _boundingboxextents) (.y _boundingboxextents) (.z _boundingboxextents)) (Variables|Default|GetMinRingSize) (Variables|Default|GetMaxRingSize)) (* 0.01 (Variables|Default|GetRingSizeMultiplier))))
  (if (Utilities|Enum|Equal(Enum) (Variables|FortGravityGunHolder|GetObjectTypeHeld) "Vehicle")
    (if (Class|AthenaFunctionLibrary|RowtoBool (Variables|Default|GetVehiclesMoveUpWhenADS))
      (Variables|UpdateLocation|SetCachedHoldUpDistanceWhileADS (Math|Float|Clamp(Float) (* (Variables|Default|GetCached_EndingScale) (ScalableFloat|GetValueatLevel (Variables|Default|GetADSUpDistanceMultiplier))) 0.0 (ScalableFloat|GetValueatLevel (Variables|Default|GetADSUpDistanceMax))) :!label _label))
    (Transformation|SetWorldRotation (Variables|FortGravityGunHolder|GetRingPivotComponent) (Math|Conversions|RotationFromXVector (Utilities|Select :Index (Variables|UpdateRotation|GetCurrentRotationAxisIndex) :Option%200 _x :Option%201 _y :Option%202 _z)) false false)
    (switch Utilities|FlowControl|Switch|SwitchonEGravityGunHolderObjectType (Variables|FortGravityGunHolder|GetObjectTypeHeld)
      (:Invalid
        _)
      (:PhysicsObject
        _)
      (:Vehicle
        (bind _returnvalue (FortPhysicsSimulationLibrary|GetPhysicsObjectSimSize (Variables|FortGravityGunHolder|GetHeldObject)))
        (Variables|CalculateLaunchVelocity|SetVehicleLaunchSpeedMultiplier (Utilities|Select :Index _returnvalue :Small (ScalableFloat|GetValueatLevel (Variables|Default|GetSizeToSpeedMultiplier_MediumVehicleMultiplier)) :Average (ScalableFloat|GetValueatLevel (Variables|Default|GetSizeToSpeedMultiplier_MediumVehicleMultiplier)) :Medium (ScalableFloat|GetValueatLevel (Variables|Default|GetSizeToSpeedMultiplier_MediumVehicleMultiplier)) :Large (ScalableFloat|GetValueatLevel (Variables|Default|GetSizeToSpeedMultiplier_LargeVehicleMultiplier)) :Huge 0.0 :Giant 0.0 :Enormous 0.0 :Invalid 0.0)))
      (:PickUp
        (Variables|CalculateLaunchVelocity|SetLaunchSpeedMultiplier (ScalableFloat|GetValueatLevel (Variables|Default|GetSizeToSpeedMultiplier_PickupMultiplier)))))
    (bind _returnvalue_2 (FortPhysicsSimulationLibrary|GetPhysicsObjectSimSize (Variables|FortGravityGunHolder|GetHeldObject)))
    (Variables|CalculateLaunchVelocity|SetLaunchSpeedMultiplier (* (Utilities|Select :Index _returnvalue_2 :Small (ScalableFloat|GetValueatLevel (Variables|Default|GetSizeToSpeedMultiplier_SmallMultiplier)) :Average (ScalableFloat|GetValueatLevel (Variables|Default|GetSizeToSpeedMultiplier_AverageMultiplier)) :Medium (ScalableFloat|GetValueatLevel (Variables|Default|GetSizeToSpeedMultiplier_MediumMultiplier)) :Large (ScalableFloat|GetValueatLevel (Variables|Default|GetSizeToSpeedMultiplier_LargeMultiplier)) :Huge 0.0 :Giant 0.0 :Enormous 0.0 :Invalid 0.0) (Math|Float|Lerp (ScalableFloat|GetValueatLevel (Variables|Default|GetSizeToSpeedMultiplier_Min)) (ScalableFloat|GetValueatLevel (Variables|Default|GetSizeToSpeedMultiplier_Max)) (Math|Curves|GetFloatValue (Variables|Default|GetSizeToSpeedMultiplier_Curve) (Math|Float|NormalizetoRange (Math|Float|Clamp(Float) (Variables|Default|GetHeldObjectMass) (ScalableFloat|GetValueatLevel (Variables|Default|GetSizeToSpeedMultiplier_SizeFloor)) (ScalableFloat|GetValueatLevel (Variables|Default|GetSizeToSpeedMultiplier_SizeCieling))) (ScalableFloat|GetValueatLevel (Variables|Default|GetSizeToSpeedMultiplier_SizeFloor)) (ScalableFloat|GetValueatLevel (Variables|Default|GetSizeToSpeedMultiplier_SizeCieling)))))))
    (else
      (Transformation|SetActorRotation (Transformation|GetActorRotation (Variables|FortGravityGunHolder|GetHeldObject)))))
  (goto _label))

(fn OnRep_HolderInfo
  (bind (_holderinfo_heldobject_2_acb32fee4c7a2835090e81b8e298d87d _holderinfo_owningplayer_4_f897963f4ce3fef8fb8d5da78d9386b3 _holderinfo_objecttype_7_d7ea2bf3487afa9161842a9b16a059c0 _holderinfo_projectileclass_10_365ba54746c6cb4c5f9d7689d70bb2ec) (Utilities|Struct|BreakGravyGoblinHolderInfo (Variables|GetHolderInfo_HeldObject_2_ACB32FEE4C7A2835090E81B8E298D87D)))
  (Utilities|IsValid _holderinfo_heldobject_2_acb32fee4c7a2835090e81b8e298d87d
    (:Is%20Valid
      (CallFunction|Initialize2 _holderinfo_heldobject_2_acb32fee4c7a2835090e81b8e298d87d _holderinfo_owningplayer_4_f897963f4ce3fef8fb8d5da78d9386b3 _holderinfo_objecttype_7_d7ea2bf3487afa9161842a9b16a059c0 _holderinfo_projectileclass_10_365ba54746c6cb4c5f9d7689d70bb2ec) ; TODO connect this later
    )))

(fn OnRep_NewHeldObject
  (if (Variables|Default|GetInitialized)
    (CallFunction|SwapinDummyObject)
    (else
      (Variables|Default|SetSwapInDummyOnInit true))))

(fn SwapInDummyObject
  (bind _dummyobject (Variables|Default|GetDummyObject))
  (bind _asbga_pickup_dummy_gravy_goblin (Utilities|Casting|CastToBGA_Pickup_Dummy_GravyGoblin _dummyobject))
  (Utilities|FlowControl|Sequence
    (:then_0
      (Actor|SetOwner (Variables|Default|GetDummyObject) (Variables|FortGravityGunHolder|GetOwningPlayer))
      (switch Utilities|FlowControl|Switch|SwitchonEGravityGunHolderObjectType (Variables|FortGravityGunHolder|GetObjectTypeHeld)
        (:Projectile
          (bind _returnvalue (AddComponent|NODE_AddStaticMeshComponent-2 :self (Variables|Default|GetDummyObject) :RelativeTransform_Location "0, 0, 0" :RelativeTransform_Rotation "0, 0, 0" :RelativeTransform_Scale "1.000000,1.000000,1.000000"))
          (Components|StaticMesh|SetStaticMesh _returnvalue (Variables|/Script/Engine.StaticMeshComponent|GetStaticMesh (Actor|GetComponentbyClass (Variables|Default|GetOriginalProjectile) "/Script/Engine.StaticMeshComponent")))
          (Collision|SetCollisionEnabled _returnvalue "NoCollision")
          (GravityGunHolderMethods|SetHeldObject (Variables|Default|GetDummyObject)))
        (:PickUp
          (Transformation|AttachActorToActor (Variables|Default|GetOriginalPickup) (Variables|Default|GetDummyObject) "None" "SnapToTarget" "KeepWorld" "KeepWorld" false)
          (Class|BGAPickupDummyGravyGoblin|AttachPickup _asbga_pickup_dummy_gravy_goblin (Utilities|Casting|CastToFortPickup (Variables|Default|GetOriginalPickup)))
          (GravityGunHolderMethods|SetHeldObject _asbga_pickup_dummy_gravy_goblin))))
    (:then_1
      (CallFunction|Attach))))

(fn TurnBackOnHoldingPawnCollision
  (bind _heldobject (Variables|FortGravityGunHolder|GetHeldObject))
  (bind _returnvalue (Actor|GetComponentbyClass _heldobject "/Script/FortniteGame.FortPhysicsObjectComponent"))
  (if (Class|AthenaFunctionLibrary|RowtoBool (Variables|Default|GetIgnoreCollisionWithPlayerHolding))
    (if (Variables|Default|GetHoldingObject)
      (Variables|Default|SetHoldingObject false)
      (Utilities|IsValid _returnvalue
        (:Is%20Valid
          (PhysicsObject|DecrementNeverSleep _returnvalue)
          (bind _asfort_pawn (Utilities|Casting|CastToFortPawn (Variables|FortGravityGunHolder|GetOwningPlayer)))
          (PhysicsObject|IgnorePawnforCollision _returnvalue _asfort_pawn false true))))))

(fn ConstructionScript
  (Variables|FortGravityGunHolder|SetOwningPlayer (Game|GetInstigator))
  (Development|PrintString "CONSTRUCTION" true true "(R=0.000000,G=0.660000,B=1.000000,A=1.000000)" 6.0))
